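; Backpack pickup effects Blueprint that adds a background particle effect when the pickup is tossed.
; On toss it runs the parent logic, then asynchronously loads the Backpack FX asset if one is set.
; The loaded particle system is spawned attached to the root, stored as the inherited background particle system, and given a maximum draw distance.
; The construction script only calls the parent.
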(event EventOnTossed
  (bind _backpack_fx (Variables|Default|GetBackpack%20FX))
  (|CallParentFunction)
  (if (not (Asset|IsAssetNull _backpack_fx))
    (bind _object (Utilities|AsyncLoadAsset _backpack_fx)
      (:Completed
        (bind _asparticle_system (Utilities|Casting|CastToParticleSystem _object))
        (bind _returnvalue (Effects|SpawnEmitterAttached _asparticle_system (Variables|Default|GetRoot)))
        (bind _output_get (Variables|FortPickupsParent|SetBackgroundParticleSystem _returnvalue))
        (LOD|SetMaxDrawDistance _output_get 5e+03)))))

(fn ConstructionScript
  (|CallParentFunction))
