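; Default pickup effect component that styles a pickup's background glow, pickup material and rare sparkle effect by rarity.
; BeginPlay spawns the glow mesh, and OnInitialized runs three init passes, adding the Rare FX mesh only for rarity index 2 and above.
; Each pass sets rarity colors and maps curve table rows, sampled at the rarity index, to scalar parameters, skipping negative values.
; Glow and pickup MIDs receive parameters only when flagged as needing initialization.
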(event EventBeginPlay
  (bind _returnvalue (SubComponent|AddChildComponent (Variables|Default|GetGlow_BG_Blueprint) "Glow_BG"))
  (bind _asstatic_mesh_component (Utilities|Casting|CastToStaticMeshComponent _returnvalue))
  (Variables|Default|SetGlow_BG_Component _asstatic_mesh_component)
  (|CallParentFunction))

(event EventOnInitialized
  (CallFunction|InitializeBG)
  (CallFunction|InitializePickupEffect)
  (CallFunction|InitializeRareFX))

(event Custom|Initialize_BG
  (bind _returnvalue (CallFunction|GetPickupRarity))
  (bind _returnvalue_2 (Math|Conversions|ToInteger(Byte) _returnvalue))
  (bind _midresult (CallFunction|GetBGGlowMID (Variables|Default|GetGlow_BG_Component)))
  (if (.bNeedsToBeInitialized (Utilities|Struct|BreakMaterialInstanceDynamicResult _midresult))
    (Rendering|Material|SetVectorParameterValue (Utilities|Struct|BreakMaterialInstanceDynamicResult _midresult) "Color" (Utilities|Array|Get(aref) (Variables|Default|GetBackgroundRarityColors) _returnvalue_2) :!class /Script/Engine.MaterialInstanceDynamic)
    (bind (_keypin _valuepin) (Utilities|Map|ForEachLoop(Map) (Variables|Default|GetBG_MID_ParametersNames)))
    (bind _outxy (DataTable|EvaluateCurveTableRow "/Game/Effects/Fort_Effects/Gameplay/Pickups/DataTables/PickupEffect_BG_MID_Parameters.PickupEffect_BG_MID_Parameters" _keypin (Math|Conversions|ToFloat(Integer) _returnvalue_2))
      (:RowFound
        (if (>= _outxy 0)
          (Rendering|Material|SetScalarParameterValue (Utilities|Struct|BreakMaterialInstanceDynamicResult _midresult) _valuepin _outxy :!class /Script/Engine.MaterialInstanceDynamic))))))

(event Custom|Initialize_PickupEffect
  (bind _returnvalue (CallFunction|GetPickupRarity))
  (bind _midresultarray (CallFunction|GetPickupMID (Variables|Default|GetPickupEffectMaterial)))
  (for _array_element _midresultarray
    (if (.bNeedsToBeInitialized (Utilities|Struct|BreakMaterialInstanceDynamicResult _array_element))
      (Rendering|Material|SetVectorParameterValue (Utilities|Struct|BreakMaterialInstanceDynamicResult _array_element) "Color" (Utilities|Array|Get(aref) (Variables|Default|GetBackgroundRarityColors) (Math|Conversions|ToInteger(Byte) _returnvalue)) :!class /Script/Engine.MaterialInstanceDynamic)
      (bind (_keypin _valuepin) (Utilities|Map|ForEachLoop(Map) (Variables|Default|GetPickup_MID_ParametersNames)))
      (bind _outxy (DataTable|EvaluateCurveTableRow "/Game/Effects/Fort_Effects/Gameplay/Pickups/DataTables/PickupEffect_Pickup_MID_Parameters.PickupEffect_Pickup_MID_Parameters" _keypin (Math|Conversions|ToFloat(Byte) _returnvalue))
        (:RowFound
          (if (>= _outxy 0)
            (Rendering|Material|SetScalarParameterValue (Utilities|Struct|BreakMaterialInstanceDynamicResult _array_element) _valuepin _outxy :!class /Script/Engine.MaterialInstanceDynamic)))))))

(event Custom|Initialize_Rare_FX
  (bind _returnvalue (CallFunction|GetPickupRarity))
  (bind _returnvalue_2 (Math|Conversions|ToInteger(Byte) _returnvalue))
  (bind _backgroundraritycolors (Variables|Default|GetBackgroundRarityColors))
  (bind _output (Utilities|Array|Get(aref) _backgroundraritycolors _returnvalue_2))
  (bind _returnvalue_3 (Math|Conversions|ToVector(LinearColor) _output))
  (if (>= (Math|Conversions|ToInteger(Byte) (CallFunction|GetPickupRarity)) 2)
    (bind _returnvalue_4 (SubComponent|AddChildComponent (Variables|Default|GetRare_FX_Blueprint) "Rare_FX"))
    (bind _asstatic_mesh_component (Utilities|Casting|CastToStaticMeshComponent _returnvalue_4))
    (bind _output_get (Variables|Default|SetRare_FX_Component _asstatic_mesh_component))
    (Rendering|Material|SetVectorParameterValueonMaterials _output_get "Color" _returnvalue_3)
    (Rendering|Material|SetVectorParameterValueonMaterials _output_get "Color_Sparkle" _returnvalue_3)
    (Rendering|Material|SetVectorParameterValueonMaterials _output_get "BaseColor" _returnvalue_3)
    (bind (_keypin _valuepin) (Utilities|Map|ForEachLoop(Map) (Variables|Default|GetRareFX_MID_ParametersNames)))
    (bind _outxy (DataTable|EvaluateCurveTableRow "/Game/Effects/Fort_Effects/Gameplay/Pickups/DataTables/PickupEffect_RareFX_MID_Parameters.PickupEffect_RareFX_MID_Parameters" _keypin (Math|Conversions|ToFloat(Byte) _returnvalue))
      (:RowFound
        (if (>= _outxy 0)
          (Rendering|Material|SetScalarParameterValueonMaterials _output_get _valuepin _outxy))))))
