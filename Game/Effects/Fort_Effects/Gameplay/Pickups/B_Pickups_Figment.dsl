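; Figment pickup Blueprint that lifts pickups tossed into water to just above the water surface.
; On toss, it resets WaterFound, then runs a short upward object trace for a BP_Athena_Water actor and teleports the pickup to that actor's water plane height.
; If that finds nothing, a downward channel trace from above handles other water volumes, using the impact height.
; BeginPlay, Tick and overlap events add no logic.
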(event EventBeginPlay
  (|CallParentFunction) ; This node is disabled and will not be called.\nDrag off pins to build functionality.
)

(event Collision|EventActorBeginOverlap
  (params (OtherActor :type Actor%20Object%20Reference)))

(event EventTick
  (params (DeltaSeconds :type Float%20%28single-precision%29)))

(event EventOnTossed
  (|CallParentFunction)
  (Variables|Default|SetWaterFound false)
  ;[1 1 1 0.5] Teleports Thrown Pickups to the top of any Water Volume they were thrown into (FORT-919246)
  (Utilities|FlowControl|Sequence
    (:then_0
      ;;[1 1 1 0.5] For Water Volumes that are instances of BP_Athena_Water
      (bind _outhit (Collision|LineTraceForObjects (Transformation|GetActorLocation) (+ (Transformation|GetActorLocation) "0.000000,0.000000,10.000000") (Utilities|Array|MakeArray "ObjectTypeQuery1")))
      (bind _asbp_athena_water (Utilities|Casting|CastToBP_Athena_Water (.HitActor (Collision|BreakHitResult _outhit))))
      (Transformation|SetActorLocation (Math|Vector|MakeVector (.x (Transformation|GetActorLocation)) (.y (Transformation|GetActorLocation)) (+ (.z (Transformation|GetWorldLocation (Variables|/Game/Athena/Environments/Blueprints/Water/BP_Athena_Water.BP_Athena_Water_C|GetWater_Plane _asbp_athena_water))) 5.0)) false true)
      (Variables|Default|SetWaterFound true))
    (:then_1
      ;;[1 1 1 0.5] Prevent the second check if the first already hit.
      (if (not (Variables|Default|GetWaterFound))
        ;;[1 1 1 0.5] For Water Volumes that aren't instances of BP_Athena_Water
        (bind (_outhit_2 _returnvalue) (Collision|LineTraceByChannel (+ (Transformation|GetActorLocation) "0.000000,0.000000,100.000000") (Transformation|GetActorLocation) "TraceTypeQuery12"))
        (if _returnvalue
          (Transformation|SetActorLocation (Math|Vector|MakeVector (.x (Transformation|GetActorLocation)) (.y (Transformation|GetActorLocation)) (+ (.z (.ImpactPoint (Collision|BreakHitResult _outhit_2))) 5.0)) false true))))))

(fn ConstructionScript
  (|CallParentFunction))
