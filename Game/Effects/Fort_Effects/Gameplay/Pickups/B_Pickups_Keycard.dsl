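; Keycard pickup Blueprint that gives its mesh a different material for each location or boss keycard variant.
; The construction script looks up the pickup's item definition in a map to get an index, which selects the material for slot 0.
; On pickup, it calls the parent, then creates a dynamic material instance and adds it to the inherited MID array.
; Both material steps run only where cosmetic events can execute.
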(event EventBeginPlay
  (|CallParentFunction) ; This node is disabled and will not be called.\nDrag off pins to build functionality.
)

(event Collision|EventActorBeginOverlap
  (params (OtherActor :type Actor%20Object%20Reference)))

(event EventTick
  (params (DeltaSeconds :type Float%20%28single-precision%29))
  (|CallParentFunction DeltaSeconds) ; This node is disabled and will not be called.\nDrag off pins to build functionality.
)

(event EventOnPickedUp
  (params (PickupTarget :type Fort%20Pawn%20Object%20Reference))
  (|CallParentFunction 0)
  (Networking|CanExecuteCosmeticEvents
    (:True
      (bind _returnvalue (Rendering|Material|CreateDynamicMaterialInstance (Variables|Default|GetPickupMesh) 0 0 "None" :!class /Script/Engine.PrimitiveComponent))
      (Utilities|Array|AddUnique (Variables|FortPickupsParent|GetMIDs) _returnvalue))))

(fn GetBlueprintPickupMesh
  (return (Variables|Default|GetPickupMesh)))

(fn ConstructionScript
  (|CallParentFunction)
  (Networking|CanExecuteCosmeticEvents
    (:True
      (Rendering|Material|SetMaterial (Variables|Default|GetPickupMesh) 0 (Utilities|Select :Index (Utilities|Map|Find (Variables|Default|GetItemDefMap) (Variables|PickupProperties|GetItemDefinition)) :Option%200 (Variables|Default|GetBaseKeycardMaterial) :Option%201 (Variables|Default|GetOilRigKeycardMaterial) :Option%202 (Variables|Default|GetIslandKeycardMaterial) :Option%203 (Variables|Default|GetAgencyKeycardMaterial) :Option%204 (Variables|Default|GetYachtKeycardMaterial) :Option%205 (Variables|Default|GetDPKeycardMaterial) :Option%206 (Variables|Default|GetCattyCornerKeycardMaterial) :Option%207 (Variables|Default|GetFortillaKeycardMaterial) :Option%208 (Variables|Default|GetDate%20Keycard%20Material) :Option%209 (Variables|Default|GetIronManKeycardMaterial) :Option%2010 (Variables|Default|GetHollowBaseKeycardMaterial))))))
