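; Parent Blueprint for dropped-item pickup visuals, covering rarity effects, mobile highlighting and pickup feedback.
; Construction sets rarity color, mesh references and bounds radius, and scales hologram timings for Athena.
; Tossing can activate background particles, attaching can lift the mesh above ground, and a timeline scales mobile selection.
; Pickup disables particles and icon, triggers material setup and spawns Campaign loot bursts; helpers drive timestamp-based hologram materials and rarity-parameterized trails.
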(event Custom|HideBackgroundAndSpotlight)

(event EventBeginPlay
  (bind _static_mesh_pickup (Variables|Default|GetStatic_Mesh_Pickup))
  ;[1 0 0] Begin Play
  (MobileIcons|SetupMobileInteractionMaterial)
  (Rendering|SetAffectDynamicIndirectLighting _static_mesh_pickup false)
  (Rendering|SetAffectDistanceFieldLighting _static_mesh_pickup false))

(event EventOnPickedUp
  (params (PickupTarget :type Fort%20Pawn%20Object%20Reference))
  (bind (_asfort_ingredient_item_definition _bsuccess) (Utilities|Casting|CastToFortIngredientItemDefinition (Variables|PickupProperties|GetItemDefinition)))
  (bind _itemdefinition (Variables|PickupProperties|GetItemDefinition))
  (bind (_asfort_resource_item_definition _bsuccess_2) (Utilities|Casting|CastToFortResourceItemDefinition _itemdefinition))
  ;[1 1 1] item picked up by user
  (Utilities|FlowControl|Sequence
    (:then_0
      (Variables|HiddenDefault|SetPickedUp true)
      (switch Utilities|FlowControl|Switch|SwitchonESubGame (Fortnite|GetSubGame)
        (:Campaign
          (if _bsuccess_2
            _
            (elif (>= (Variables|FortPickupsParent|GetPickupRarityLevel) 1)
              (Utilities|FlowControl|Delay 0.25)
              (Effects|SpawnEmitteratLocation "/Game/Effects/Fort_Effects/Effects/LootContainers/P_Loot_Effect_02.P_Loot_Effect_02" (Transformation|GetWorldLocation (Variables|Default|GetSkeletalOrStaticMeshAssetPrimitive)))
              (else
                (if _bsuccess
                  _)))))))
    (:then_1
      (CallFunction|SetActiveBackgroundParticleSystem false)
      (CallFunction|SetVisibleMobileInteractIcon false)
      (CallFunction|OnPickedUpMIDSetup))))

(event EventEnableBacchusHighlight
  (bind _rootcomponent (Variables|Transformation|GetRootComponent))
  ;[1 0 0] Mobile highlighting
  (Networking|CanExecuteCosmeticEvents
    (:True
      (bind (_direction _lerpobject) (Timeline|MobileSelectedTL Play)
        (:Update
          (Transformation|SetRelativeScale3D (Variables|Transformation|GetRootComponent) (Math|Conversions|ToVector(Float) _lerpobject)))
        (:Finished
          (switch Utilities|FlowControl|Switch|SwitchonETimelineDirection _direction
            (:Forward
              (Transformation|SetRelativeScale3D _rootcomponent (Variables|MobileIcons|GetMobileSelectedScale)))
            (:Backward
              (Transformation|SetRelativeLocation (Variables|Transformation|GetRootComponent) "0, 0, 0" false false)
              (Transformation|SetRelativeScale3D _rootcomponent "1.000000,1.000000,1.000000"))))))))

(event EventOnTossed
  ;[0 0 0] Activate particle system if tossed and it should exist
  (|CallParentFunction)
  (if (Variables|FortPickupsParent|GetbActivateRarityParticleSystems)
    (CallFunction|SetActiveBackgroundParticleSystem true)
    (CallFunction|SetupViewDistances)))

(event EventDisableBacchusHighlight
  (bind _rootcomponent (Variables|Transformation|GetRootComponent))
  ;[1 0 0] Mobile highlighting
  (Networking|CanExecuteCosmeticEvents
    (:True
      (if (not (Variables|HiddenDefault|GetPickedUp))
        (Timeline|MobileSelectedTL Reverse)))))

(event EventOnAttached
  (bind _returnvalue (Class|FortPickup|GetMovementCapsuleHalfHeight))
  (bind _skeletalorstaticmeshassetprimitive (Variables|Default|GetSkeletalOrStaticMeshAssetPrimitive))
  (bind (_origin _boxextent) (Collision|GetComponentBounds _skeletalorstaticmeshassetprimitive))
  (bind (_boxextent_x _boxextent_y _boxextent_z) (Math|Vector|BreakVector _boxextent))
  (bind _returnvalue_2 (- _boxextent_z _returnvalue))
  (bind (_returnvalue_x _returnvalue_y _returnvalue_z) (Math|Vector|BreakVector (Transformation|GetWorldLocation _skeletalorstaticmeshassetprimitive)))
  ;[1 1 1] OnAttached
  (|CallParentFunction)
  (Networking|CanExecuteCosmeticEvents
    (:True
      ;;[1 1 1] If necessary, raise the mesh so that the bottom of its bounding box is at least flush with the ground
      (bind _itemdefinition (Variables|PickupProperties|GetItemDefinition)
        (:then
          (Class|FortItemFunctionLibrary|GetPickupMeshTransform _itemdefinition
            (:IsIdentity
              (Utilities|Casting|CastToFortPickup (Actor|GetAttachParentActor))
              (if (> _returnvalue_2 0.0)
                (Transformation|SetWorldLocation (Variables|Default|GetSkeletalOrStaticMeshAssetPrimitive) (Math|Vector|MakeVector _returnvalue_x _returnvalue_y (+ _returnvalue_2 _returnvalue_z)) false false)))))
        (:else)))))

(fn DestroyBackgroundVisualComponents)

(fn getViewDistanceQuality
  (if (Editor|InEditororPIE)
    (bind _returnvalue (Development|GetConsoleVariableIntValue "sg.ViewDistanceQuality"))
    (return _returnvalue)
    (else
      (return (Settings|GetViewDistanceQuality (Settings|GetFortGameUserSettings))))))

(fn ScaleHologramTimingsForPvP
  (switch Utilities|FlowControl|Switch|SwitchonESubGame (Fortnite|GetSubGame)
    (:Athena
      (Variables|HologramSetup|SetPickedUpDelayBeforeGlow (* (Variables|HologramSetup|GetPickedUpDelayBeforeGlow) (LocalVariables|GetPVP_TimeScale)))
      (Variables|HologramSetup|SetPickedUpFadeInTime (* (Variables|HologramSetup|GetPickedUpFadeInTime) (LocalVariables|GetPVP_TimeScale)))
      (Variables|HologramSetup|SetPickedUpHoldTime (* (Variables|HologramSetup|GetPickedUpHoldTime) (LocalVariables|GetPVP_TimeScale)))
      (Variables|HologramSetup|SetPickedUpFadeOutTime (* (Variables|HologramSetup|GetPickedUpFadeOutTime) (LocalVariables|GetPVP_TimeScale)))
      (Variables|HologramSetup|SetPickedUpSweepSpeed (* (Variables|HologramSetup|GetPickedUpSweepSpeed) (LocalVariables|GetPVP_TimeScale))))))

(fn SetActiveBackgroundParticleSystem
  (params (Active :type Boolean) (Reset :type Boolean))
  (bind _backgroundparticlesystem (Variables|FortPickupsParent|GetBackgroundParticleSystem)
    (:then
      (Components|Activation|SetActive _backgroundparticlesystem Active Reset))
    (:else)))

(fn SetHiddenBackgroundVisualComponents
  (params (Hidden :type Boolean)))

(fn SetHologramPickedUpParams
  (params (Tier%200 :type Boolean))
  (bind _returnvalue (Utilities|Time|GetGameTimeinSeconds))
  (bind _mids (Variables|FortPickupsParent|GetMIDs))
  (if Tier%200
    ;[1 1 1] For common
    (for _array_element _mids
      (Utilities|IsValid _array_element
        (:Is%20Valid
          (Rendering|Material|SetScalarParameterValue _array_element "TimestampB" _returnvalue :!class /Script/Engine.MaterialInstanceDynamic)
          (Rendering|Material|SetScalarParameterValue _array_element "TransitionLength_Out" (Variables|HologramSetup|GetPickedUpFadeOutTime) :!class /Script/Engine.MaterialInstanceDynamic)
          (Rendering|Material|SetScalarParameterValue _array_element "Sweep Speed" 0.0 :!class /Script/Engine.MaterialInstanceDynamic)
          (Rendering|Material|SetScalarParameterValue _array_element "Random" (* (Math|Float|Fraction _returnvalue) -1) :!class /Script/Engine.MaterialInstanceDynamic))))
    (else
      ;[1 1 1] All others
      (for _array_element_2 _mids
        (Utilities|IsValid _array_element_2
          (:Is%20Valid
            (Rendering|Material|SetScalarParameterValue _array_element_2 "TimestampA" (Utilities|Time|GetGameTimeinSeconds) :!class /Script/Engine.MaterialInstanceDynamic)
            (Rendering|Material|SetScalarParameterValue _array_element_2 "TransitionLength_In" (Variables|HologramSetup|GetPickedUpFadeInTime) :!class /Script/Engine.MaterialInstanceDynamic)
            (Rendering|Material|SetScalarParameterValue _array_element_2 "TimestampB" (Math|Float|float+float (Variables|HologramSetup|GetPickedUpHoldTime) (Variables|HologramSetup|GetPickedUpFadeInTime) (Utilities|Time|GetGameTimeinSeconds)) :!class /Script/Engine.MaterialInstanceDynamic)
            (Rendering|Material|SetScalarParameterValue _array_element_2 "TransitionLength_Out" (Variables|HologramSetup|GetPickedUpFadeOutTime) :!class /Script/Engine.MaterialInstanceDynamic)
            (Rendering|Material|SetScalarParameterValue _array_element_2 "Sweep Speed" (Variables|HologramSetup|GetPickedUpSweepSpeed) :!class /Script/Engine.MaterialInstanceDynamic)
            (Rendering|Material|SetScalarParameterValue _array_element_2 "Random" (* (Math|Float|Fraction (Utilities|Time|GetGameTimeinSeconds)) -1) :!class /Script/Engine.MaterialInstanceDynamic))))))
  (return))

(fn SetupViewDistances
  (Utilities|FlowControl|Sequence
    (:then_0
      (bind _returnvalue (CallFunction|GetViewDistanceQuality))
      (LocalVariables|Setl_viewQuality _returnvalue))
    (:then_2
      (return (LocalVariables|Getl_viewQuality)))))

(fn SetVisibleMobileInteractIcon
  (params (Visible :type Boolean))
  (if Visible
    ;[1 1 1] Create
    (bind _mobileinteracticon (Variables|Default|GetMobileInteractIcon)
      (:then
        (Rendering|SetVisibility _mobileinteracticon true false))
      (:else
        (Variables|MobileIcons|GetMobileInteractionMaterial
          (:then
            (bind _returnvalue (AddComponent|NODE_AddStaticMeshComponent-3 :RelativeTransform_Location (Variables|MobileIcons|GetMobileInteractIconLocation) :RelativeTransform_Rotation_Yaw (* (Variables|HiddenDefault|GetRandom%20Rotation) -1.0) :RelativeTransform_Scale (Variables|MobileIcons|GetMobileInteractIconScale)))
            (bind _output_get (Variables|Default|SetMobileInteractIcon _returnvalue))
            (Rendering|Material|SetMaterial _output_get 0 (Variables|MobileIcons|GetMobileInteractionMaterial)))
          (:else))))
    (else
      ;[1 1 1] Destroy
      (bind _mobileinteracticon_2 (Variables|Default|GetMobileInteractIcon)
        (:then
          (Rendering|SetVisibility _mobileinteracticon_2 false false)
          (Components|DestroyComponent _mobileinteracticon_2))
        (:else)))))

(fn SpawnPickedUpTrailPS
  (bind _returnvalue (Effects|SpawnEmitterAttached (Variables|FortPickupsParent|GetPickedUpTrailPS) (Variables|Transformation|GetRootComponent)))
  (LocalVariables|SetTrailPS _returnvalue)
  (Effects|Components|ParticleSystem|SetFloatParameter (LocalVariables|GetTrailPS) "SPU Scale" (Utilities|Array|Get(acopy) (Variables|HiddenDefault|GetSparkle%20Spawn%20Rate%20%28Picked%20Up%29) (Variables|FortPickupsParent|GetPickupRarityLevel)))
  (Effects|Components|ParticleSystem|SetFloatParameter (LocalVariables|GetTrailPS) "SparkleSphereRadius" (Variables|HiddenDefault|GetComponentRadiusScaled))
  (Effects|Components|ParticleSystem|SetFloatParameter (LocalVariables|GetTrailPS) "Lifetime" (Utilities|Array|Get(acopy) (Variables|HiddenDefault|GetLifetime%20%28Picked%20Up%29) (Variables|FortPickupsParent|GetPickupRarityLevel)))
  (Effects|Components|ParticleSystem|SetVectorParameter (LocalVariables|GetTrailPS) "Color" (Math|Conversions|ToVector(LinearColor) (Variables|FortPickupsParent|GetPickupRarityColor)))
  (return))

(fn ConstructionScript
  (CallFunction|SetRarityColor (Variables|HologramSetup|GetOutline%20Rarity%20Colors))
  (CallFunction|ScaleHologramTimingsforPvP)
  (CallFunction|SetupPickupMesh)
  (bind _output_get (Variables|Default|SetStatic_Mesh_Pickup (Variables|FortPickupsParent|GetStaticMeshComponent)))
  (Rendering|SetReceivesDecals _output_get false)
  (bind _output_get_2 (Variables|Default|SetSkeletal_Mesh_Pickup (Variables|FortPickupsParent|GetSkelMeshComponent)))
  (Rendering|SetReceivesDecals _output_get_2 false)
  (bind _output_get_3 (Variables|Default|SetSkeletalOrStaticMeshAssetPrimitive (Variables|FortPickupsParent|GetMeshPrimitiveComponent)))
  (Rendering|SetReceivesDecals _output_get_3 false)
  (Variables|HiddenDefault|SetComponentRadiusScaled (* (.SphereRadius (Collision|GetComponentBounds _output_get_3)) (Variables|HiddenDefault|GetComponent%20Radius%20Multiplier)))
  (Variables|HiddenDefault|SetComponent%20Radius (.SphereRadius (Collision|GetComponentBounds _output_get_3))))
