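# Niagara ambient effect for rare pickups: soft upward-drifting sprite light shafts tinted by User.RarityColor.
# Shafts001 spawns continuously in a cylinder, with one biased random value (CurveRand) linking height, radius and speed, so taller shafts sit nearer the center and rise faster.
# Shafts vary randomly in brightness, fade in and out, widen with scalability distance and dim slightly at night.
# The cone-based Streaks emitter is disabled.
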NiagaraSystem /Game/Effects/Fort_Effects/Gameplay/Pickups/Effects/NS_Pickup_RareAmbient
dependencies {usedDataInterfaces [&NiagaraDataInterfaceCurve], usedDynamicInputs [], usedModules [&SystemState]}
emitters {
 Shafts001 {
  emitterUpdateScript [
   &EmitterState()
   &SpawnRate(SpawnRate=7, Use Spawn Probability=false)
  ]
  particleSpawnScript [
   /SetVariables(Particles.CurveRand:NiagaraFloat=&FloatFromCurve(
     FloatCurve=DataInterface() {propertyValues {Curve {keys [
         {interpMode=RCIM_Cubic, tangentMode=RCTM_User, time=0, value=0}
         {arriveTangent=0.7649, interpMode=RCIM_Cubic, leaveTangent=0.7649, tangentMode=RCTM_User, time=0.8, value=0.2}
         {arriveTangent=1.7117, interpMode=RCIM_Cubic, leaveTangent=1.7117, tangentMode=RCTM_User, time=1, value=1}
        ]}}}
     CurveIndex=&RandomRangeFloat(Minimum=0, Maximum=1)
    )) {isSetParameters=true}
   &InitializeParticle(
    Color=&MultiplyLinearColor(
     Linear Color=$User.RarityColor
     Scale Factor=&MakeFloat4FromVectorAndFloat(
      "Vector (RGB)"=&VectorFromFloat(Value=&Lerp_Float(
        A=0.25
        B=4
        Alpha=&FloatFromCurve(
         FloatCurve=DataInterface() {propertyValues {Curve {keys [
             {interpMode=RCIM_Cubic, tangentMode=RCTM_User, time=0, value=0}
             {arriveTangent=0.8992, interpMode=RCIM_Cubic, leaveTangent=0.8992, tangentMode=RCTM_User, time=0.8, value=0.25}
             {arriveTangent=2.5764, interpMode=RCIM_Cubic, leaveTangent=2.5764, tangentMode=RCTM_User, time=1, value=1}
            ]}}}
         CurveIndex=&RandomRangeFloat(Minimum=0, Maximum=1)
        )
       ))
      "Float (Alpha)"=1
     )
    )
    Lifetime Max=1.5
    Lifetime Mode=::NewEnumerator1
    Sprite Rotation Mode=::NewEnumerator0
    Sprite Size=&Lerp_Vector2(A=<5 125>, B=<5 50>, Alpha=$Particles.CurveRand)
    Sprite Size Mode=::NewEnumerator4
    UsePositionOffset=false
    Write Material Random=true
   )
   &ShapeLocation(
    Cylinder Height=&Lerp_Float(A=30, B=10, Alpha=$Particles.CurveRand)
    Cylinder Radius=&Lerp_Float(A=10, B=45, Alpha=$Particles.CurveRand)
    MaskSpawnGroup=false
    Offset=<0 0 -30>
    Offset Mode=::NewEnumerator0
    Override Seed=false
    Shape Primitive=::NewEnumerator1
    Surface Only=false
    UseOwnerScale=true
   )
   &AddVelocity(Rotation Coordinate Space=::World, Velocity=&Lerp_Vector(A=<0 0 25>, B=<0 0 2>, Alpha=$Particles.CurveRand))
  ]
  particleUpdateScript [
   &ParticleState()
   &ScaleColor(
    Scale Alpha=&FloatFromCurve(
     FloatCurve=DataInterface() {propertyValues {Curve {keys [
         {interpMode=RCIM_Cubic, tangentMode=RCTM_Break, time=0, value=0}
         {interpMode=RCIM_Cubic, tangentMode=RCTM_Break, time=0.4, value=1}
         {interpMode=RCIM_Cubic, tangentMode=RCTM_Break, time=1, value=0}
        ]}}}
     Scale Curve=&RandomRangeFloat(Minimum=0.5, Maximum=0.9)
    )
    Scale RGB=&VectorFromFloat(Value=&Lerp_Float(
      A=1
      B=10
      Alpha=&FloatFromCurve(
       FloatCurve=DataInterface() {propertyValues {Curve {keys [
           {interpMode=RCIM_Cubic, tangentMode=RCTM_Break, time=0, value=0}
           {arriveTangent=0.3549, interpMode=RCIM_Cubic, leaveTangent=0.3549, tangentMode=RCTM_User, time=0.6, value=0.2}
           {arriveTangentWeight=0.1278, interpMode=RCIM_Cubic, tangentMode=RCTM_Break, time=0.9, value=1}
          ]}}}
       CurveIndex=&RandomRangeFloat(Minimum=0, Maximum=1)
      )
     ))
    ScaleA=true
    ScaleRGB=true
   )
   &ScaleSpriteSize(
    Scale Factor=&MakeVector2D(
     X=&FloatFromCurve(
      FloatCurve=DataInterface() {propertyValues {Curve {keys [{interpMode=RCIM_Cubic, tangentMode=RCTM_Break, time=0, value=0.1}, {interpMode=RCIM_Cubic, tangentMode=RCTM_Break, time=1, value=1}]}}}
      Scale Curve=/Niagara/DynamicInputs/Scalability/ScalabilityDistanceBasedFloat(
       Curve:NiagaraDataInterfaceCurve=DataInterface() {propertyValues {Curve {keys [
           {interpMode=RCIM_Cubic, tangentMode=RCTM_User, time=0, value=1}
           {arriveTangent=3.3503, interpMode=RCIM_Cubic, leaveTangent=3.3503, tangentMode=RCTM_User, time=0.7, value=2}
           {interpMode=RCIM_Cubic, tangentMode=RCTM_User, time=1, value=4}
          ]}}}
       CurveIndex:NiagaraFloat=$Emitter.Scalability.DistanceFraction
       Bias:NiagaraFloat=0
       Scale:NiagaraFloat=1
      )
     )
     Y=&FloatFromCurve(FloatCurve=DataInterface() {propertyValues {Curve {keys [
          {interpMode=RCIM_Cubic, leaveTangentWeight=0.5079, tangentMode=RCTM_Break, time=0, value=0.4}
          {interpMode=RCIM_Cubic, tangentMode=RCTM_Break, time=1, value=1}
         ]}}})
    )
    Scale Sprite Size Mode=::NewEnumerator2
   )
   &Drag(Drag=0.5, Use Linear Drag=true, Use Rotational Drag=true)
   &SolveForcesAndVelocity(Clamp Velocity=false, Limit Acceleration=false)
   &ScaleColor(
    Scale Alpha=&FNTimeOfDayScalar(Night Scale=0.8, Morning Scale=1, Evening Scale=1)
    ScaleA=true
    ScaleRGB=false
   ) {module=ScaleColor001}
  ]
  rendererClasses [&NiagaraSpriteRendererProperties]
  renderers [{rendererClass=&NiagaraSpriteRendererProperties}]
 }
 Streaks {
  bEnabled=false
  emitterUpdateScript [
   &EmitterState()
   &SpawnRate(SpawnRate=12, Use Spawn Probability=false)
  ]
  particleSpawnScript [
   &InitializeParticle(
    Color Maximum=<rgba 1.2213 3.0848 8 1>
    Color Minimum=<rgba 0.5086 0.7136 0.8958 1>
    Color Mode=::NewEnumerator2
    Lifetime Max=1.5
    Lifetime Min=0.75
    Lifetime Mode=::NewEnumerator1
    Sprite Size Max=<1.5 200>
    Sprite Size Min=<0.5 100>
    Sprite Size Mode=::NewEnumerator2
    UsePositionOffset=false
    Write Material Random=true
   )
   /Niagara/Modules/Spawn/Location/ConeLocation(
    Cone Angle=5
    Cone Axis:Vector3f=<0 0 1>
    Cone Length=30
    Enable Wedge=true
    MaskSpawnGroup=false
    Module.Override Randomness=false
    Offset:Vector3f=<0 0 -10>
    Wedge Horizontal Angle=25
    Wedge Vertical Angle=60
   )
   &ShapeLocation(Cone Angle=5, Cone Length=30, MaskSpawnGroup=false, Override Seed=false, Shape Primitive=::NewEnumerator5, UseOwnerScale=true) {enabled=false}
   &AddVelocityFromPoint(Module.Use Falloff Distance=false, Origin Offset=<0 0 -30>, Velocity Strength=50)
   /SetVariables(Particles.RandAlpha:NiagaraFloat=&RandomRangeFloat(Minimum=0.1, Maximum=0.4)) {isSetParameters=true}
  ]
  particleUpdateScript [
   &ParticleState()
   &SolveForcesAndVelocity(Clamp Velocity=false, Limit Acceleration=false)
   &ScaleColor(
    Scale Alpha=&FloatFromCurve(
     FloatCurve=DataInterface() {propertyValues {Curve {keys [<at 0 0 RCIM_Cubic>,<at 0.5 1.1 RCIM_Cubic>,<at 1 0 RCIM_Cubic>]}}}
     Scale Curve=$Particles.RandAlpha
    )
    ScaleA=true
    ScaleRGB=true
   )
  ]
  rendererClasses [&NiagaraSpriteRendererProperties]
  renderers [{rendererClass=&NiagaraSpriteRendererProperties}]
 }
}
summary {
 systemName=NS_Pickup_RareAmbient
 emitters [
  {emitterName=Shafts001, rendererClasses [&NiagaraSpriteRendererProperties]}
  {bEnabled=false, emitterName=Streaks, rendererClasses [&NiagaraSpriteRendererProperties]}
 ]
 userVariables [
  User.Radius {type=NiagaraFloat, defaultValue=1, description=""}
  User.RarityColor {type=LinearColor, description="", defaultValue {struct=LinearColor, value=<rgba 1 1 1 1>}}
 ]
}
