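# Unlit, translucent, two-sided material for curl mesh particles in pickup effects, with Nanite usage also enabled.
# Particle color RGB feeds MF_FX_EmissiveAdjustments to produce the glow color.
# Particle alpha multiplies the T_Active_Reload_Hit_Marker_Ramp texture to form a fade mask.
# A dynamic parameter offsets that ramp's V coordinate, letting the particle system slide the mask along the mesh.
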Material /Game/Effects/Fort_Effects/Gameplay/Pickups/Materials/M_Curl_mesh
graph {
 properties {
  bUsedWithMeshParticles=true
  bUsedWithNanite=true
  blendMode=BLEND_Translucent
  pixelDepthOffsetMode=PDOM_Legacy
  shadingModel=MSM_Unlit
  twoSided=true
  usedShadingModels=Unlit
 }
 nodes [
  Constant() {r=0.5}
  n1 = ParticleColor()
  MaterialFunctionCall(Emissive=@n1.RGB) {materialFunction=&MF_FX_EmissiveAdjustments}
  n2 = TextureCoordinate()
  n3 = AppendVector(A=Constant(), B=DynamicParameter() {defaultValue=<rgba 0 0 0 1>}.Param1)
  Multiply(A=@n1.A, B=TextureSample(UVs=Add(A=@n2, B=@n3)) {texture=/Game/Effects/Fort_Effects/Textures/T_Active_Reload_Hit_Marker_Ramp}.RGB)
 ]
}
