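# Unlit translucent god-ray material for particle sprites and Niagara mesh or sprite particles, filed under pickup effects.
# Color blends particle color with a masked glow tint and an optional curve-atlas gradient, then passes through emissive adjustments.
# A horizontally panned noise texture, shaped by UV edge and vertical fades, optional Fresnel and particle alpha, drives opacity.
# Depth fade softens intersections, except on the mobile shading path.
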Material /Game/Effects/Fort_Effects/Gameplay/Pickups/Materials/M_GodRays
graph {
 properties {
  bUsedWithNiagaraMeshParticles=true
  bUsedWithNiagaraSprites=true
  bUsedWithParticleSprites=true
  blendMode=BLEND_Translucent
  pixelDepthOffsetMode=PDOM_Legacy
  shadingModel=MSM_Unlit
  usedShadingModels=Unlit
 }
 nodes [
  n1 = ParticleColor()
  n2 = SphereMask(
   A=TextureCoordinate()
   B=MaterialFunctionCall(X=Constant() {r=0.5}, Y=Constant() {r=0.7}) {materialFunction=&MakeFloat2}.Result
   Radius=Constant() {r=0.5}
   Hardness=Constant()
  )
  n3 = LinearInterpolate(
   A=LinearInterpolate(A=@n1.RGB, B=VectorParameter() {parameterName=Glow, defaultValue=<rgba 15 9.7031 0.45 0>}.RGB, Alpha=Power(Base=@n2) {constExponent=3})
   B=@n1.RGB
   Alpha=ScalarParameter() {parameterName=JustParticleColor}
  )
  n4 = TextureCoordinate()
  n5 = CurveAtlasRowParameter(CurveTime=OneMinus(None=ComponentMask(None=@n4) {r=true})) {
   atlas=/Game/Athena/LinearColorCurves/VFX_CurveAtlas
   curve=/Game/Accessories/FORT_Backpacks/Backpack_F_MED_Quarrel/FX/Curve_Quarrel_Auroras_02
   defaultValue=29
   group=Curve
   parameterName=Curve
  }
  n6 = LinearInterpolate(A=Constant3Vector() {constant=<rgba 1 1 1 0>}, B=@n5, Alpha=ScalarParameter() {defaultValue=1, parameterName=Color Curve Intensity})
  MaterialFunctionCall(Emissive=LinearInterpolate(A=@n3, B=Multiply(A=@n6, B=@n3), Alpha=ScalarParameter() {parameterName=Use Color Curve})) {materialFunction=&MF_FX_EmissiveAdjustments}
  n7 = DynamicParameter() {paramNames [PanningSpeed,Param2,Param3,Param4]}
  n8 = ScalarParameter() {parameterName=Panning Speed}
  n9 = ScalarParameter() {parameterName=UsePanningSpeed}
  n10 = Panner(
   Coordinate=Multiply(A=ScalarParameter() {defaultValue=1, parameterName=TextureTile}, B=TextureCoordinate() {vTiling=0})
   Speed=MaterialFunctionCall(X=LinearInterpolate(A=@n7.PanningSpeed, B=@n8, Alpha=@n9)) {materialFunction=&MakeFloat2}.Result
  )
  n11 = Multiply(
   A=Power(
    Base=TextureSampleParameter2D(UVs=@n10) {parameterName=Tex, texture=&T_NoisePack_01}.R
    Exp=ScalarParameter() {defaultValue=1.35, parameterName=Texture_Power}
   ) {constExponent=1.35}
   B=ScalarParameter() {defaultValue=3, parameterName=Texture_Multiply}
  ) {constB=3}
  n12 = TextureCoordinate()
  n13 = Multiply(A=Multiply(A=@n12, B=OneMinus(None=@n12))) {constB=4}
  n14 = Multiply(A=ComponentMask(None=@n13) {r=true}, B=ComponentMask(None=@n13) {g=true})
  n15 = ComponentMask(None=@n12) {g=true}
  n16 = Multiply(A=@n15, B=OneMinus(None=@n15))
  n17 = ScalarParameter() {parameterName=FadeBottom}
  n18 = Multiply(
   A=Saturate(None=Multiply(A=@n14) {constB=4})
   B=Power(Base=LinearInterpolate(A=@n15, B=@n16, Alpha=@n17), Exp=ScalarParameter() {defaultValue=2, parameterName=FadeTop})
  )
  n19 = MaterialFunctionCall(Exponent=Constant() {r=3}) {materialFunction=&MF_Fresnel_FX}
  n20 = ScalarParameter() {defaultValue=1, parameterName=UseFresnel}
  n21 = Multiply(A=@n1.A, B=Multiply(A=Multiply(A=@n11, B=@n18), B=LinearInterpolate(B=@n19.Fresnel, Alpha=@n20) {constA=1}))
  n22 = DepthFade(Opacity=@n21, FadeDistance=ScalarParameter() {parameterName=DepthFade_Distance})
  ShadingPathSwitch(Default=@n22, Deferred=@n22, Mobile=@n21) {desc=Migrated from FeatureLevelSwitch}
 ]
}
