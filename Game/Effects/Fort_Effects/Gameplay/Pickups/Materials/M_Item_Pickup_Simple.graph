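# Translucent unlit material that draws an item pickup effect: glow with flickering corona, vertical light pillar and rising sparkles.
# Vertex colors split one mesh into these parts, which world position offset shapes into camera-offset sprites, thickened spline and noise-drifting, curve-sized sparkle sprites.
# Brightness follows time of day, and opacity uses depth and camera-distance fades.
# Four scene-depth taps add a glowing outline near the actor on the non-mobile shading path.
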Material /Game/Effects/Fort_Effects/Gameplay/Pickups/Materials/M_Item_Pickup_Simple
graph {
 properties {blendMode=BLEND_Translucent, pixelDepthOffsetMode=PDOM_Legacy, shadingModel=MSM_Unlit, twoSided=true, usedShadingModels=Unlit}
 nodes [
  n1 = Multiply(A=NamedRerouteUsage()) {constB=0.3}
  n2 = Sine(None=@n1) {period=0.075}
  n3 = Multiply(A=Multiply(A=Sine(None=@n1) {period=0.155}, B=Sine(None=@n1) {period=0.2}), B=Sine(None=@n1) {period=0.43})
  n4 = Add(A=Multiply(A=Multiply(A=@n2, B=@n3), B=ScalarParameter() {defaultValue=0.05, parameterName=CoronaFlickerScale})) {constB=0.75}
  n5 = TextureSample(UVs=Saturate(None=MaterialFunctionCall(Texture Scale=@n4) {materialFunction=&ScaleUVsByCenter}.UVs)) {texture=/Game/Athena/Items/Accolades/Materials/T_UI_HaloCorona}
  n6 = Multiply(A=@n5.RGB) {constB=20}
  Desaturation(None=@n6, Fraction=Constant() {r=0.5})
  n7 = MaterialFunctionCall!(Center Location, 0-1 UVs, XY scale=ScalarParameter() {defaultValue=120, parameterName=GlowSize}, OverridePivot, "Normals (Optional)") {materialFunction=&Sprite}
  n8 = MaterialFunctionCall(Offset Amount=ScalarParameter() {defaultValue=40, parameterName=CameraOffset}) {materialFunction=/Engine/Functions/Engine_MaterialFunctions02/WorldPositionOffset/CameraOffset}
  n9 = ScalarParameter() {defaultValue=70, parameterName=PillarWidth}
  n10 = MaterialFunctionCall(WidthBase=@n9, WidthTip=@n9) {materialFunction=&SplineThicken}
  n11 = ComponentMask(None=TextureCoordinate()) {g=true}
  n12 = Multiply(
   A=Saturate(None=LinearInterpolate(Alpha=@n11) {constA=1, constB=-1})
   B=Multiply(A=ScalarParameter() {defaultValue=200, parameterName=PillarHeight}, B=Constant3Vector() {constant=<rgba 0 0 1 0>})
  )
  n13 = ScalarParameter() {defaultValue=40, group=Scale - Mesh Sprite, parameterName=SparkleSize}
  n14 = Multiply(A=Add(A=ScalarParameter() {parameterName=Random}, B=Time()), B=ScalarParameter() {defaultValue=1, group=Time, parameterName=SparkleSpeed})
  n15 = Multiply(A=WorldPosition().XYZ) {constB=0.02}
  n16 = VertexColor()
  n17 = LinearInterpolate(
   A=Add(
    A=NamedRerouteDeclaration(None=@n14) {name=TimeMultiplied, nodeColor=<rgba 0.875 0 1 1>}
    B=MaterialFunctionCall(f3=@n15) {materialFunction=&AddComponents}.f3
   )
   B=@n16.B
   Alpha=ScalarParameter() {defaultValue=0.4, group=Time, parameterName=Modulate Speed By Vertex Color}
  )
  n18 = ComponentMask(None=Frac(None=Add(A=@n17, B=@n16.B))) {b=true}
  n19 = CurveAtlasRowParameter(CurveTime=@n18) {
   atlas=/Game/Athena/LinearColorCurves/VFX_CurveAtlas
   curve=/Game/Athena/LinearColorCurves/Curve_Burst_Up_Slow_Down
   defaultValue=5
   group=Scale - Mesh Sprite
   parameterName="Sprite Size Over Life Curve (Expects Greyscale)"
  }
  n20 = LinearInterpolate(Alpha=@n19.B) {constA=-0.5}
  n21 = LinearInterpolate(
   A=@n13
   B=Multiply(A=@n13, B=Saturate(None=@n20))
   Alpha=ScalarParameter() {defaultValue=1, group=Scale - Mesh Sprite, parameterName=Modulate Size Over Life}
  )
  n22 = Multiply(A=VertexColor().A, B=@n21)
  n23 = ScalarParameter() {defaultValue=1, group=Scale - Mesh Sprite, parameterName=Modulate Size By Alpha}
  n24 = Subtract(A=WorldPosition() {worldPositionShaderOffset=WPT_CameraRelative}.XYZ, B=ObjectPositionWS())
  n25 = Normalize(VectorInput=VertexNormalWS())
  n26 = ScalarParameter() {defaultValue=1, group=Scale - Expansion, parameterName=Expansion Intensity}
  n27 = Add(
   A=MaterialFunctionCall(Center Location=WorldPosition().XYZ, XY scale=LinearInterpolate(A=@n21, B=@n22, Alpha=@n23)) {materialFunction=&Sprite}.World position offset
   B=LinearInterpolate(A=Normalize(VectorInput=@n24), B=Multiply(A=@n25, B=@n26), Alpha=ScalarParameter() {group=Scale - Expansion, parameterName=Expansion Lerp})
  ) {constB=0}
  n28 = VectorParameter() {group=Movement - Magnitude And Vector, parameterName="Movement Overall Direction (WorldSpace)", defaultValue=<rgba 0 0 1 0>}
  n29 = Multiply(
   A=Multiply(A=@n18, B=ScalarParameter() {defaultValue=25, group=Movement - Magnitude And Vector, parameterName=Magnitude Multiplier}) {constB=25}
   B=@n28.RGB
  )
  n30 = Subtract(A=WorldPosition().XYZ, B=ActorPositionWS())
  n31 = Multiply(A=ScalarParameter() {defaultValue=8, group=Movement - Noise, parameterName=Noise Speed}, B=Time())
  n32 = VectorNoise(World Position=Divide(A=Subtract(A=@n30, B=@n31), B=ScalarParameter() {defaultValue=50, group=Movement - Noise, parameterName=Noise Scale})) {noiseFunction=VNF_VectorALU}
  n33 = ScalarParameter() {defaultValue=12, group=Movement - Noise, parameterName=Noise Intensity}
  LinearInterpolate(
   A=LinearInterpolate(A=Add(A=@n7.World position offset, B=@n8.Clamped Camera Offset), B=Add(A=@n10.WorldPositionOffset, B=@n12), Alpha=VertexColor().G)
   B=Add(A=Add(A=@n27, B=@n29), B=Multiply(A=@n32, B=@n33))
   Alpha=NamedRerouteUsage()
  )
  n34 = ScalarParameter() {defaultValue=0.6, group=TOD, parameterName=TOD Morning Light}
  n35 = ScalarParameter() {defaultValue=1, group=TOD, parameterName=TOD Day Light}
  n36 = ScalarParameter() {defaultValue=0.6, group=TOD, parameterName=TOD Evening Light}
  n37 = ScalarParameter() {defaultValue=0.3, group=TOD, parameterName=TOD Night Light}
  n38 = VectorParameter() {group=General Light, parameterName=Color Light, sortPriority=5, defaultValue=<rgba 0.5885 0.4639 0.3462 0>}
  n39 = Add(A=@n38.RGB) {constB=0.1}
  n40 = Multiply(A=Power(Base=@n5.A), B=ScalarParameter() {defaultValue=1, parameterName=CoronaBrightness})
  n41 = VertexColor()
  n42 = ComponentMask(None=NamedRerouteDeclaration(None=Multiply(A=@n40, B=@n41.R)) {name=CoronaOpacity, nodeColor=<rgba 0 0.0625 1 1>}) {g=true}
  n43 = Distance(A=TextureCoordinate(), B=Constant() {r=0.5})
  n44 = ScalarParameter() {defaultValue=2, group=General Light, parameterName=Glow Size Light}
  n45 = Saturate(None=OneMinus(None=Multiply(A=@n43, B=@n44) {constB=2}))
  n46 = Multiply(A=@n45, B=@n45)
  n47 = Multiply(A=@n46) {constB=4}
  n48 = Add(A=Multiply(A=NamedRerouteUsage()) {constB=2}, B=@n47)
  n49 = LinearInterpolate(
   A=Add(A=Multiply(A=@n6, B=@n42), B=ShadingPathSwitch(Default=@n48, Mobile=@n47))
   B=ScalarParameter() {defaultValue=1.5, parameterName=SparkleBrightness}
   Alpha=NamedRerouteUsage()
  ) {constB=1.5}
  MaterialFunctionCall(Emissive=Multiply(
    A=MaterialFunctionCall(Value_Morning=@n34, Value_Day=@n35, Value_Evening=@n36, Value_Night=@n37) {materialFunction=&Scalar_Time_of_Day}.Result
    B=Multiply(A=@n39, B=@n49)
   )) {materialFunction=&MF_FX_EmissiveAdjustments}
  n50 = Multiply(A=NamedRerouteUsage(), B=ScalarParameter() {defaultValue=0.75, group=Outline, parameterName=Outline Opacity})
  n51 = Distance(
   A=MaterialFunctionCall() {materialFunction=&WorldPositionBehindTranslucency}.Result
   B=ActorPositionWS()
  )
  n52 = ScalarParameter() {defaultValue=-60, group=Outline, parameterName=Outline Effect Radius}
  n53 = Saturate(None=Divide(A=Add(A=@n51, B=@n52), B=ScalarParameter() {defaultValue=-20, group=Outline, parameterName=Outline Falloff}))
  n54 = Multiply(A=@n46, B=ScalarParameter() {defaultValue=0.25, parameterName=GlowOpacity})
  n55 = ScalarParameter() {defaultValue=1, parameterName=GlowPower}
  n56 = NamedRerouteUsage()
  n57 = Multiply(A=OneMinus(None=@n53), B=Add(A=Power(Base=@n54, Exp=@n55), B=Multiply(A=@n56) {constB=0.01})) {constA=1}
  n58 = ScalarParameter() {defaultValue=25, group=General, parameterName=Depth Fade Distance, sortPriority=5}
  n59 = Multiply(A=ShadingPathSwitch(Default=Add(A=@n50, B=@n57), Mobile=@n57), B=QualitySwitch(Default=DepthFade(FadeDistance=@n58), Low=Constant() {r=1}))
  n60 = LinearInterpolate(B=ScalarParameter() {defaultValue=0.25, parameterName=SplineOpacity}, Alpha=VertexColor().G) {constA=1}
  n61 = ScalarParameter() {defaultValue=0.25, parameterName=SparkleOpacity}
  n62 = TextureSample() {texture=/Game/Effects/Fort_Effects/Textures/Flares/T_GlowStar_5pt}
  Multiply(
   A=LinearInterpolate(A=Multiply(A=@n59, B=@n60), B=Multiply(A=@n61, B=@n62.RGB), Alpha=NamedRerouteUsage())
   B=MaterialFunctionCall!(Distance=ScalarParameter() {parameterName=CameraDistanceFade}) {materialFunction=/Game/Effects/Fort_Effects/Materials/_Functions/MF_CameraDistanceFade}.Result
  )
  n63 = SceneDepth()
  n64 = Multiply(A=ScalarParameter() {defaultValue=1.5, group=Outline, parameterName=Outline Width}, B=SceneTexelSize()) {constA=2}
  n65 = Constant()
  n66 = AppendVector(A=ComponentMask(None=@n64) {desc=x, r=true}, B=@n65)
  n67 = Subtract(A=@n63.R, B=SceneDepth!(OffsetFraction=@n66) {inputMode=OffsetFraction}.R)
  n68 = AppendVector(A=@n65, B=ComponentMask(None=@n64) {desc=y, g=true}) {desc=y+}
  n69 = Subtract(A=@n63.R, B=SceneDepth(OffsetFraction=@n68) {inputMode=OffsetFraction}.R)
  n70 = Constant() {r=-1}
  n71 = Subtract(A=@n63.R, B=SceneDepth(OffsetFraction=Multiply(A=@n66, B=@n70)) {inputMode=OffsetFraction}.R)
  n72 = Subtract(A=@n63.R, B=SceneDepth(OffsetFraction=Multiply(A=@n70, B=@n68)) {inputMode=OffsetFraction}.R)
  n73 = Divide(A=Add(A=Add(A=@n67, B=@n69), B=Add(A=@n71, B=@n72))) {constB=-100}
  NamedRerouteDeclaration(None=Multiply(A=Clamp(None=@n73) {desc=depth outline}, B=@n53)) {name=GlowingOutline, nodeColor=<rgba 0.5234 1 0 1>}
  n74 = VertexColor()
  NamedRerouteDeclaration(None=OneMinus(None=Max(A=@n74.R, B=@n74.R))) {name=SparklesMask, nodeColor=<rgba 1 0.1172 0 1>}
 ]
}
