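# Unlit translucent mesh-particle material for loot pickup star effects, with emissive from particle color or a Color parameter via MF_FX_EmissiveAdjustments.
# A UseWPO-blended offset pushes vertices along a dynamic-parameter direction and spins them around the object pivot.
# A UV-derived, Frac-wrapped vector, shifted by a dynamic parameter, varies direction and rotation axis.
# Particle red drives distance (squared, scaled by V) and spin angle, green the spread, alpha passes through.
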Material /Game/Effects/Fort_Effects/Gameplay/Pickups/Materials/M_Loot_Star
graph {
 properties {
  bUsedWithMeshParticles=true
  bUsedWithNiagaraMeshParticles=true
  blendMode=BLEND_Translucent
  pixelDepthOffsetMode=PDOM_Legacy
  shadingModel=MSM_Unlit
  usedShadingModels=Unlit
 }
 nodes [
  n1 = DynamicParameter()
  n2 = Normalize(VectorInput=AppendVector(A=AppendVector(A=@n1.Param1, B=@n1.Param1), B=@n1.Param3))
  n3 = TextureCoordinate()
  n4 = Add(
   A=AppendVector(A=@n3, B=ComponentMask(None=@n3) {r=true})
   B=Multiply(A=ComponentMask(None=@n3) {g=true}, B=Constant3Vector() {constant=<rgba 0 20 0 0>})
  )
  n5 = ConstantBiasScale(None=Frac(None=Add(A=@n1.Param4, B=@n4))) {bias=-0.5, scale=2}
  n6 = ParticleColor()
  n7 = Multiply(A=@n5, B=@n6.G)
  n8 = Multiply(A=Normalize(VectorInput=Add(A=@n2, B=@n7)), B=Multiply(A=@n6.R, B=@n6.R))
  n9 = TextureCoordinate()
  n10 = Multiply(A=Add(A=ComponentMask(None=@n9) {g=true}, B=Constant() {r=0.25}), B=Constant() {r=2})
  n11 = Normalize(VectorInput=@n5)
  n12 = MaterialFunctionCall() {materialFunction=&ObjectPivotPoint}
  n13 = WorldPosition()
  LinearInterpolate(
   A=Constant()
   B=Add(
    A=Multiply(A=@n8, B=@n10)
    B=RotateAboutAxis(NormalizedRotationAxis=@n11, RotationAngle=@n6.R, PivotPoint=@n12.Object Pivot Location, Position=@n13.XYZ)
   )
   Alpha=ScalarParameter() {defaultValue=1, parameterName=UseWPO}
  )
  MaterialFunctionCall(Emissive=StaticSwitchParameter(True=ParticleColor().RGB, False=VectorParameter() {parameterName=Color, defaultValue=<rgba 3 1.5 0.5 1>}.RGB) {parameterName=UseForParticleColor}) {materialFunction=&MF_FX_EmissiveAdjustments}
  Reroute(None=@n6.A)
 ]
}
