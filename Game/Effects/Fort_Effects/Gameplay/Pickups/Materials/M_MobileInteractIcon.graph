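# Translucent unlit material for a world-space mobile interact icon that faces the camera and pulses.
# World position offset aligns the mesh to camera, pushes it toward the viewer by the interact extents, and adds a distance-scaled cosine bob.
# A radial timer ring, built from rotated UVs and two sphere gradients, is added over the tinted, channel-packed icon texture.
# Opacity pulses ring and center separately, with depth fade and near-camera fade.
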Material /Game/Effects/Fort_Effects/Gameplay/Pickups/Materials/M_MobileInteractIcon
graph {
 properties {bUsedWithNanite=true, blendMode=BLEND_Translucent, pixelDepthOffsetMode=PDOM_Legacy, shadingModel=MSM_Unlit, usedShadingModels=Unlit}
 nodes [
  n1 = Normalize(VectorInput=Subtract(A=CameraPositionWS(), B=ObjectPositionWS()))
  n2 = Cosine(None=Subtract(A=Time(), B=ScalarParameter() {parameterName=PulseStartTime}))
  n3 = CameraPositionWS()
  n4 = WorldPosition()
  n5 = Subtract(
   A=MaterialFunctionCall(Vector 3=Subtract(A=@n3, B=@n4.XYZ)) {materialFunction=/Engine/Functions/Engine_MaterialFunctions02/Utility/VectorLength}.V3 Length
   B=ScalarParameter() {defaultValue=150, parameterName=NearFadeStart}
  )
  n6 = ScalarParameter() {defaultValue=350, parameterName=NearFadeEnd}
  n7 = Multiply(
   A=Add(A=@n2, B=Constant() {r=1})
   B=Multiply(A=ScalarParameter() {defaultValue=40, parameterName=CameraOffsetPulseAmount}, B=Divide(A=@n5, B=@n6))
  )
  n8 = WorldPosition()
  n9 = Constant3Vector() {constant=<rgba 0 -1 0 0>}
  n10 = Constant3Vector() {constant=<rgba 0 0 -1 0>}
  n11 = Normalize(VectorInput=Subtract(A=CameraPositionWS(), B=ObjectPositionWS()))
  n12 = Abs(None=@n11)
  n13 = VectorParameter() {parameterName=Interact Extents}
  Add(
   A=Add(
    A=Multiply(A=@n1, B=@n7)
    B=MaterialFunctionCall!(WorldPosition=@n8.XYZ, PivotLocation, WS Normals, Custom Object Basis 1=@n9, Custom Object Basis 2=@n10, Custom Object Basis 3) {materialFunction=/Engine/Functions/Engine_MaterialFunctions02/WorldPositionOffset/AlignMeshToTheCamera}.World Position Offset
   )
   B=Multiply(A=@n11, B=DotProduct(A=@n12, B=@n13.RGB))
  )
  n14 = Subtract(A=Time(), B=ScalarParameter() {parameterName=TimerStartTime})
  n15 = ScalarParameter() {defaultValue=1, parameterName=TimerDuration}
  n16 = Subtract(A=MaterialFunctionCall(UVs=TextureCoordinate(), "Rotation Angle (0-1)"=ScalarParameter() {defaultValue=-0.5, parameterName=Origin_0-1}) {materialFunction=&CustomRotator}.Rotated Values) {constB=0.5}
  n17 = Add(
   A=Clamp(None=Divide(A=@n14, B=@n15) {constA=0.75})
   B=OneMinus(None=MaterialFunctionCall(Vector or UVs=@n16) {materialFunction=&VectorToRadialValue}.Vector Converted to Angle)
  )
  n18 = Clamp(None=Multiply(A=ScalarParameter() {defaultValue=1, parameterName=OuterRadius_0-1}) {constB=0.5}) {maxDefault=0.5}
  n19 = Clamp(None=Multiply(A=MaterialFunctionCall!(UVs, CenterPosition, Radius=@n18) {materialFunction=/Engine/Functions/Engine_MaterialFunctions01/Gradient/SphereGradient-2D}.Result 0-1) {constB=3})
  n20 = Clamp(None=Multiply(A=ScalarParameter() {defaultValue=0.8, parameterName=InnerRadius_0-1}) {constB=0.5}) {maxDefault=0.495}
  n21 = Clamp(None=Multiply(A=MaterialFunctionCall(Radius=@n20) {materialFunction=&SphereGradient-2D}.Result 0-1) {constB=3})
  n22 = ComponentMask(None=Multiply(A=Floor(None=@n17), B=Subtract(A=@n19, B=@n21))) {r=true}
  n23 = Multiply(A=ScalarParameter() {parameterName=TimerOpacity}, B=AppendVector(A=AppendVector(A=@n22, B=@n22), B=@n22))
  n24 = VectorParameter() {parameterName=Color, defaultValue=<rgba 1 1 1 1>}
  n25 = Multiply(A=@n24.RGB) {constB=1.5}
  n26 = TextureCoordinate()
  n27 = ComponentMask(None=@n26) {g=true}
  n28 = TextureSampleParameter2D(UVs=AppendVector(A=OneMinus(None=@n27), B=ComponentMask(None=@n26) {r=true})) {parameterName=Icon, samplerType=SAMPLERTYPE_LinearColor, texture=/Game/Athena/HUD/Art/BacchusHudIcons/Interact_Hit_Mobile_L}
  Add(A=@n23, B=LinearInterpolate(A=LinearInterpolate(A=@n25, Alpha=@n28.B) {constB=0}, Alpha=@n28.R))
  n29 = SphereMask(A=TextureCoordinate(), B=Constant() {r=0.5}, Radius=Constant() {r=0.45}, Hardness=Constant() {r=0.95})
  n30 = OneMinus(None=@n29)
  n31 = Power(Base=Saturate(None=ConstantBiasScale(None=@n2) {bias=1.2}), Exp=ScalarParameter() {defaultValue=5, parameterName=Exponent})
  n32 = LinearInterpolate(
   A=ScalarParameter() {defaultValue=0.75, parameterName=Opacity Low Ring}
   B=ScalarParameter() {defaultValue=1, parameterName=Opacity High Ring}
   Alpha=@n31
  )
  n33 = LinearInterpolate(
   A=ScalarParameter() {defaultValue=0.1, parameterName=Opacity Low Inner}
   B=ScalarParameter() {defaultValue=1, parameterName=Opacity High Inner}
   Alpha=@n31
  )
  n34 = Multiply(A=Multiply(A=@n24.A, B=@n28.G), B=Add(A=Multiply(A=@n30, B=@n32), B=Multiply(A=@n29, B=@n33)))
  n35 = MaterialFunctionCall(Vector 3=Subtract(A=CameraPositionWS(), B=WorldPosition().XYZ)) {materialFunction=&VectorLength}
  n36 = ScalarParameter() {defaultValue=150, parameterName=NearFadeStart}
  Multiply(
   A=DepthFade(Opacity=Add(A=@n23, B=@n34), FadeDistance=ScalarParameter() {defaultValue=40, parameterName=Depth-Fade Distance})
   B=Divide(A=Subtract(A=@n35.V3 Length, B=@n36), B=ScalarParameter() {defaultValue=350, parameterName=NearFadeEnd})
  )
  ObjectPositionWS()
  ObjectPositionWS()
 ]
}
