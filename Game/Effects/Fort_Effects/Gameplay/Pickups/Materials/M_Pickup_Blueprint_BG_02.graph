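# Additive unlit material for the blueprint pickup background effect, combining holographic grid and scanlines, an animated star and noisy fresnel beams.
# Vertex color channels branch between layers on one mesh, tinted by a shared color.
# Default shading path adds radius-masked depth-edge outlines from neighboring SceneDepth samples; mobile omits them.
# World position offset camera-aligns or collapses parts by vertex color and IsLegendaryRarity, pulling some toward the camera off mobile.
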Material /Game/Effects/Fort_Effects/Gameplay/Pickups/Materials/M_Pickup_Blueprint_BG_02
graph {
 properties {blendMode=BLEND_Additive, pixelDepthOffsetMode=PDOM_Legacy, shadingModel=MSM_Unlit, usedShadingModels=Unlit}
 nodes [
  n1 = Distance(
   A=MaterialFunctionCall() {materialFunction=&WorldPositionBehindTranslucency}.Result
   B=ActorPositionWS()
  )
  n2 = ScalarParameter() {defaultValue=-50, parameterName=EffectRadius}
  n3 = Clamp(None=Divide(A=Add(A=@n1, B=@n2), B=ScalarParameter() {defaultValue=-20, parameterName=Falloff}))
  n4 = VectorParameter() {parameterName=Color, defaultValue=<rgba 0.5885 0.4639 0.3462 0>}
  n5 = MaterialFunctionCall!(Alpha=VertexColor().G, Color Input 0=NamedRerouteUsage(), Color Input 1=NamedRerouteUsage()) {materialFunction=/Engine/Functions/Engine_MaterialFunctions02/Utility/DynamicBranch}
  n6 = ScalarParameter() {defaultValue=1, parameterName=HoloIntensity}
  n7 = Panner(Coordinate=Add(A=TextureCoordinate() {uTiling=4, vTiling=0}, B=NamedRerouteUsage())) {speedX=0.2}
  n8 = Multiply(
   A=TextureSample(UVs=NamedRerouteDeclaration(None=@n7) {name=RayUVs, nodeColor=<rgba 0.3828 1 0 1>}) {texture=&T_NoisePack_01}.R
   B=Constant() {r=1.5}
  ) {constB=3}
  n9 = Constant() {r=0.5}
  n10 = ComponentMask(None=TextureCoordinate()) {g=true}
  n11 = Saturate(None=ConstantBiasScale(None=@n10) {bias=-0.5, scale=2})
  n12 = Multiply(A=@n10, B=OneMinus(None=Multiply(A=@n11, B=@n11)))
  n13 = Multiply(
   A=MaterialFunctionCall(Brightness=Constant() {r=0.75}, Exponent=Constant() {r=12}) {materialFunction=&MF_Fresnel_FX}.Fresnel
   B=Multiply(A=Add(A=@n8, B=@n9) {constB=0.7}, B=Multiply(A=@n12, B=@n12))
  )
  n14 = ScalarParameter() {defaultValue=2, parameterName=BeamIntesity}
  n15 = Multiply(
   A=@n4.RGB
   B=MaterialFunctionCall(Alpha=VertexColor().A, Color Input 0=Multiply(A=@n5.Result, B=@n6), Color Input 1=Multiply(A=@n13, B=@n14)) {materialFunction=&DynamicBranch}.Result
  )
  n16 = Multiply(A=@n15, B=OneMinus(None=@n3))
  n17 = SceneDepth()
  n18 = Multiply(A=Constant() {r=1}, B=SceneTexelSize()) {constA=2}
  n19 = Constant()
  n20 = AppendVector(A=ComponentMask(None=@n18) {desc=x, r=true}, B=@n19)
  n21 = AppendVector(A=@n19, B=ComponentMask(None=@n18) {desc=y, g=true}) {desc=y+}
  n22 = Add(
   A=Subtract(A=@n17.R, B=SceneDepth!(OffsetFraction=@n20) {inputMode=OffsetFraction}.R)
   B=Subtract(A=@n17.R, B=SceneDepth(OffsetFraction=@n21) {inputMode=OffsetFraction}.R)
  )
  n23 = Constant() {r=-1}
  n24 = Multiply(A=@n20, B=@n23)
  n25 = Multiply(A=@n23, B=@n21)
  n26 = Add(
   A=Subtract(A=@n17.R, B=SceneDepth(OffsetFraction=@n24) {inputMode=OffsetFraction}.R)
   B=Subtract(A=@n17.R, B=SceneDepth(OffsetFraction=@n25) {inputMode=OffsetFraction}.R)
  )
  n27 = Clamp(None=Divide(A=Add(A=@n22, B=@n26), B=Constant() {r=-100})) {desc=depth outline}
  n28 = ComponentMask(None=TextureCoordinate()) {g=true}
  n29 = Multiply(A=Multiply(A=@n28, B=@n28) {constB=2}, B=Constant() {r=50})
  n30 = Multiply(A=@n4.RGB, B=Multiply(A=Multiply(A=@n3, B=@n27), B=Clamp(None=@n29)))
  MaterialFunctionCall(Emissive=ShadingPathSwitch(Default=Add(A=@n16, B=@n30), Mobile=@n15) {desc=Migrated from FeatureLevelSwitch}) {materialFunction=&MF_FX_EmissiveAdjustments}
  MaterialFunctionCall(
   Fade_Start_Distance=ScalarParameter() {defaultValue=1400, parameterName=End Fade}
   Fade_End_Distance=ScalarParameter() {defaultValue=1200, parameterName=Start Fade}
  ) {materialFunction=&MF_Distance_SyncedFade_FX}
  n31 = ComponentMask(None=TextureCoordinate()) {r=true}
  n32 = Multiply(A=@n31, B=ScalarParameter() {defaultValue=2, parameterName=Gradient Tile A})
  n33 = Multiply(A=@n31, B=ScalarParameter() {defaultValue=2.5, parameterName=Gradient Tile B})
  n34 = NamedRerouteDeclaration(None=ScalarParameter() {parameterName=Random}) {name=Random, nodeColor=<rgba 0.5703 1 0 1>}
  n35 = ScalarParameter() {defaultValue=2.8, parameterName=Panner Speed}
  n36 = Add(A=Time(), B=@n34)
  n37 = Add(
   A=AppendVector(A=AppendVector(A=@n32, B=@n33), B=Add(A=@n31, B=@n34))
   B=Multiply(A=Constant3Vector() {constant=<rgba 0.0786 -0.047 0.0436 0>}, B=Multiply(A=@n35, B=@n36))
  )
  n38 = Frac(None=Multiply(A=Sine(None=@n37)))
  n39 = MaterialFunctionCall(Float3=Subtract(A=Multiply(A=@n38) {constB=2})) {materialFunction=&BreakOutFloat3Components}
  n40 = ScalarParameter() {defaultValue=3, parameterName=Mask Brightness}
  n41 = Add(A=VertexColor().R, B=ScalarParameter() {defaultValue=-0.03, parameterName=Slide})
  n42 = ScalarParameter() {defaultValue=10, parameterName=Center Exponent}
  n43 = Add(
   A=Multiply(A=Multiply(A=@n39.R, B=@n39.R), B=Multiply(A=@n39.B, B=@n40))
   B=Multiply(A=Power(Base=@n41, Exp=@n42), B=ScalarParameter() {defaultValue=2, parameterName=Center Brightness})
  )
  n44 = ScalarParameter() {defaultValue=-0.7, parameterName=Ceil Bias}
  n45 = Multiply(A=Ceil(None=Add(A=@n43, B=@n44)), B=ScalarParameter() {defaultValue=3, parameterName=Ceil Brightness})
  n46 = ScalarParameter() {defaultValue=0.5, parameterName=Style Lerp}
  n47 = VertexColor()
  n48 = ScalarParameter() {defaultValue=4, parameterName=Fake Fresnel Exponent}
  NamedRerouteDeclaration(None=Multiply(A=LinearInterpolate(A=@n43, B=@n45, Alpha=@n46), B=Power(Base=@n47.R, Exp=@n48))) {name=SphericalStar, nodeColor=<rgba 0 1 0.7656 1>}
  n49 = Add(A=Time() {period=40}, B=NamedRerouteUsage())
  n50 = MaterialFunctionCall!(
   Number of Scanlines
   Scanline Speed
   "Scanlines Thickness (Bias)"
   Scanlines Dark Value
   Scanlines Light Value
   Block Opacity=MaterialFunctionCall!(WPO Intensity A, WPO Intensity B, Time=@n49, Bias) {materialFunction=/Game/Effects/Fort_Effects/Effects/Atlas/Materials/MF_Atlas_HologramWiggle_01}.Intense 0-1
   UVs
   Time=@n49
   Block Repeat
  ) {materialFunction=/Game/Effects/Fort_Effects/Effects/Atlas/Materials/MF_Atlas_Scanlines_01}
  n51 = ScalarParameter() {defaultValue=800, parameterName=Near Camera Fade}
  n52 = OneMinus(None=Power(Base=MaterialFunctionCall(Distance=@n51) {materialFunction=&Near_Camera_Fade}.Result) {constExponent=3})
  n53 = AppendVector(A=ScalarParameter() {defaultValue=0.0138, parameterName=Grid U}, B=ScalarParameter() {defaultValue=0.0083, parameterName=Grid V})
  n54 = ObjectRadius()
  n55 = Panner(Coordinate=Multiply(A=TextureCoordinate(), B=Multiply(A=@n53, B=@n54)), Time=@n49) {bFractionalPart=true, speedX=0.134, speedY=0.0366}
  n56 = ConstantBiasScale(None=Add(A=Sine(None=@n49) {period=1.338}, B=Sine(None=@n49) {period=1.618})) {bias=3.5, scale=0.2}
  n57 = Add(
   A=Multiply(
    A=TextureSample(UVs=@n55) {samplerType=SAMPLERTYPE_LinearColor, texture=/Game/Gadgets/Assets/AirStrike/Textures/T_Airstrike_Grid_01}.R
    B=Power(Base=@n56) {constExponent=3}
   )
   B=ScalarParameter() {defaultValue=0.17, parameterName=BG Boost}
  )
  n58 = LinearInterpolate(
   A=ScalarParameter() {defaultValue=0.1, parameterName=Use Holographic Elements Not scale Value}
   B=Add(A=Multiply(A=@n50.Secondary Lines, B=@n52), B=Multiply(A=@n50.Result, B=@n57))
   Alpha=ScalarParameter() {defaultValue=1, parameterName=Use Holographic Elements}
  )
  n59 = Distance(A=TextureCoordinate(), B=Constant() {r=0.5})
  n60 = Constant() {r=2}
  n61 = Saturate(None=OneMinus(None=Multiply(A=@n59, B=@n60) {constB=2}))
  n62 = Multiply(A=@n61, B=@n61)
  n63 = Constant() {r=1}
  n64 = DepthFade(FadeDistance=ScalarParameter() {defaultValue=25, parameterName=Depth Fade Distance})
  NamedRerouteDeclaration(None=Multiply(A=Multiply(A=@n58, B=@n62) {constA=1}, B=QualitySwitch(Default=@n63, Low=@n63, High=@n64, Medium=@n64))) {name=HolographicMesh, nodeColor=<rgba 0.6172 0 1 1>}
  n65 = ScalarParameter() {defaultValue=1, parameterName=Master Opaticy}
  n66 = DepthFade(Opacity=@n65, FadeDistance=ScalarParameter() {parameterName=Master DEpth fade})
  ShadingPathSwitch(Default=QualitySwitch(Default=@n65, Low=@n65, High=@n66, Medium=@n66, Epic=@n66), Mobile=@n65)
  n67 = MaterialFunctionCall(Alpha=VertexColor().A, Color Input 0=Subtract(A=ActorPositionWS(), B=WorldPosition().XYZ), Color Input 1=Constant() {r=1}) {materialFunction=&DynamicBranch}
  n68 = MaterialFunctionCall(
   Alpha=VertexColor().A
   Color Input 0=MaterialFunctionCall!(
    WorldPosition
    PivotLocation
    WS Normals
    Custom Object Basis 1
    Custom Object Basis 2=VectorParameter() {parameterName=RotationA, defaultValue=<rgba 0 1 0 1>}.RGB
    Custom Object Basis 3
   ) {materialFunction=/Engine/Functions/Engine_MaterialFunctions02/WorldPositionOffset/AlignMeshToTheCamera}.World Position Offset
   Color Input 1=ScalarParameter() {parameterName=CamFacing}
  ) {materialFunction=&DynamicBranch}
  n69 = CameraPositionWS()
  n70 = ActorPositionWS()
  n71 = Multiply(A=Normalize(VectorInput=Subtract(A=@n69, B=@n70)), B=ScalarParameter() {defaultValue=-10, parameterName=Camera Facing Push})
  n72 = MaterialFunctionCall(
   Alpha=VertexColor().G
   Color Input 0=LinearInterpolate(A=@n67.Result, B=@n68.Result) {constAlpha=0}
   Color Input 1=LinearInterpolate(A=Add(A=@n68.Result, B=@n71) {constB=0}, B=@n67.Result) {constAlpha=0}
  ) {materialFunction=&DynamicBranch}
  n73 = ActorPositionWS()
  n74 = WorldPosition()
  n75 = MaterialFunctionCall(
   Alpha=ScalarParameter() {parameterName=IsLegendaryRarity}
   Color Input 0=MaterialFunctionCall(Alpha=VertexColor().A, Color Input 0=@n72.Result, Color Input 1=Subtract(A=@n73, B=@n74.XYZ)) {materialFunction=&DynamicBranch}.Result
   Color Input 1=@n72.Result
  ) {materialFunction=&DynamicBranch}
  n76 = Subtract(A=Add(A=@n75.Result, B=WorldPosition().XYZ), B=CameraPositionWS())
  n77 = MaterialFunctionCall(
   Alpha=VertexColor().A
   Color Input 0=Constant()
   Color Input 1=Multiply(A=Normalize(VectorInput=@n76), B=ScalarParameter() {defaultValue=-90, parameterName=MoveTowardsCamera})
  ) {materialFunction=&DynamicBranch}
  n78 = CameraPositionWS()
  n79 = ObjectPositionWS()
  n80 = Saturate(None=Divide(A=Distance(A=@n78, B=@n79), B=ScalarParameter() {defaultValue=300, parameterName=NearDistanceClamp}))
  ShadingPathSwitch(Default=Add(A=@n75.Result, B=LinearInterpolate(B=@n77.Result, Alpha=@n80)), Mobile=@n75.Result) {desc=Migrated from FeatureLevelSwitch}
 ]
}
