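# Additive unlit glow material for pickup effects, combining a fresnel noise beam, holographic glow with animated band and procedural star in one mesh, selected by vertex-color dynamic branches.
# Default shading masks the glow by scene distance from the actor and adds a four-tap scene-depth outline; mobile skips both.
# World position offset switches between pivot-collapse and camera-facing by distance fade, vertex color and legendary rarity, adding a camera-ward push off mobile.
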Material /Game/Effects/Fort_Effects/Gameplay/Pickups/Materials/M_Pickup_Glow_BG
graph {
 properties {
  bUsedWithNiagaraMeshParticles=true
  blendMode=BLEND_Additive
  pixelDepthOffsetMode=PDOM_Legacy
  shadingModel=MSM_Unlit
  usedShadingModels=Unlit
 }
 nodes [
  Constant() {r=0.35}
  n1 = Constant() {r=0.75}
  n2 = Constant() {r=12}
  n3 = TextureCoordinate() {uTiling=4, vTiling=0}
  n4 = NamedRerouteUsage()
  n5 = NamedRerouteDeclaration(None=Panner(Coordinate=Add(A=@n3, B=@n4)) {speedX=0.2}) {name=RayUVs, nodeColor=<rgba 0.3828 1 0 1>}
  n6 = Add(
   A=Multiply(A=TextureSample(UVs=@n5) {texture=&T_NoisePack_01}.R, B=Constant() {r=1.5}) {constB=3}
   B=Constant() {r=0.5}
  ) {constB=0.7}
  n7 = ComponentMask(None=TextureCoordinate()) {g=true}
  n8 = Saturate(None=ConstantBiasScale(None=@n7) {bias=-0.5, scale=2})
  n9 = Multiply(A=@n7, B=OneMinus(None=Multiply(A=@n8, B=@n8)))
  n10 = Multiply(A=@n9, B=@n9)
  n11 = Multiply(
   A=Multiply(
    A=MaterialFunctionCall(Brightness=@n1, Exponent=@n2) {materialFunction=&MF_Fresnel_FX}.Fresnel
    B=Multiply(A=@n6, B=@n10)
   )
   B=ScalarParameter() {defaultValue=2, parameterName=BeamIntesity}
  )
  n12 = VertexColor()
  LinearInterpolate(B=@n11, Alpha=@n12.A)
  n13 = Subtract(A=ActorPositionWS(), B=WorldPosition().XYZ)
  n14 = Constant() {r=1}
  n15 = VertexColor()
  LinearInterpolate(A=@n13, B=@n14, Alpha=@n15.A)
  n16 = ScalarParameter() {defaultValue=1, parameterName=Opacity}
  LinearInterpolate(
   A=@n16
   B=DepthFade(Opacity=@n16, FadeDistance=ScalarParameter() {defaultValue=15, group=DepthFade, parameterName=FadeDistance})
   Alpha=ScalarParameter() {group=DepthFade, parameterName=UseDF}
  )
  n17 = Distance(
   A=MaterialFunctionCall() {materialFunction=&WorldPositionBehindTranslucency}.Result
   B=ActorPositionWS()
  )
  n18 = ScalarParameter() {defaultValue=-50, parameterName=EffectRadius}
  n19 = Clamp(None=Divide(A=Add(A=@n17, B=@n18), B=ScalarParameter() {defaultValue=-20, parameterName=Falloff}))
  n20 = VectorParameter() {parameterName=Color, defaultValue=<rgba 0.5885 0.4639 0.3462 0>}
  n21 = MaterialFunctionCall!(Alpha=VertexColor().G, Color Input 0=NamedRerouteUsage(), Color Input 1=NamedRerouteUsage()) {materialFunction=/Engine/Functions/Engine_MaterialFunctions02/Utility/DynamicBranch}
  n22 = ScalarParameter() {defaultValue=1, parameterName=HoloIntensity}
  n23 = Multiply(
   A=@n20.RGB
   B=MaterialFunctionCall(Alpha=@n12.A, Color Input 0=Multiply(A=@n21.Result, B=@n22), Color Input 1=@n11) {materialFunction=&DynamicBranch}.Result
  )
  n24 = Multiply(A=@n23, B=OneMinus(None=@n19))
  n25 = SceneDepth()
  n26 = Multiply(A=ScalarParameter() {defaultValue=1, parameterName=OutlineWidth}, B=SceneTexelSize()) {constA=2}
  n27 = Constant()
  n28 = AppendVector(A=ComponentMask(None=@n26) {desc=x, r=true}, B=@n27)
  n29 = AppendVector(A=@n27, B=ComponentMask(None=@n26) {desc=y, g=true}) {desc=y+}
  n30 = Add(
   A=Subtract(A=@n25.R, B=SceneDepth!(OffsetFraction=@n28) {inputMode=OffsetFraction}.R)
   B=Subtract(A=@n25.R, B=SceneDepth(OffsetFraction=@n29) {inputMode=OffsetFraction}.R)
  )
  n31 = Constant() {r=-1}
  n32 = Multiply(A=@n28, B=@n31)
  n33 = Multiply(A=@n31, B=@n29)
  n34 = Add(
   A=Subtract(A=@n25.R, B=SceneDepth(OffsetFraction=@n32) {inputMode=OffsetFraction}.R)
   B=Subtract(A=@n25.R, B=SceneDepth(OffsetFraction=@n33) {inputMode=OffsetFraction}.R)
  )
  n35 = Clamp(None=Divide(A=Add(A=@n30, B=@n34), B=Constant() {r=-100})) {desc=depth outline}
  n36 = ComponentMask(None=TextureCoordinate()) {g=true}
  n37 = Multiply(A=Multiply(A=@n36, B=@n36) {constB=2}, B=Constant() {r=50})
  n38 = Multiply(A=@n20.RGB, B=Multiply(A=Multiply(A=@n19, B=@n35), B=Clamp(None=@n37)))
  MaterialFunctionCall(Emissive=ShadingPathSwitch(Default=Add(A=@n24, B=@n38), Mobile=@n23) {desc=Migrated from FeatureLevelSwitch}) {materialFunction=&MF_FX_EmissiveAdjustments}
  n39 = Multiply(A=Add(A=Time() {period=40}, B=NamedRerouteUsage()), B=ScalarParameter() {defaultValue=0.85, parameterName=BandSpeed}) {constB=2}
  n40 = Clamp(None=OneMinus(None=Cosine(None=@n39)))
  n41 = MaterialFunctionCall(UVs=TextureCoordinate(), "Rotation Angle (0-1)"=Constant() {r=0.39}) {materialFunction=&CustomRotator}
  n42 = MaterialFunctionCall!(
   Width=LinearInterpolate(Alpha=@n40) {constA=0.4, constB=0.15}
   Sharpness=Constant() {r=0.5}
   Offset
   Direction Switch=StaticBool() {value=true}
   Compare
   Input Coordinates=@n41.Rotated Values
  ) {materialFunction=/Engine/Functions/Engine_MaterialFunctions03/Procedurals/GeneratedBand}
  Multiply(A=@n42.Result) {constA=1}
  n43 = TextureCoordinate()
  n44 = ScalarParameter() {defaultValue=2, parameterName=Gradient Tile A}
  n45 = ScalarParameter() {parameterName=Random}
  n46 = Multiply(A=ScalarParameter() {defaultValue=2.8, parameterName=Panner Speed}, B=Add(A=Time(), B=@n45))
  n47 = TextureSample(UVs=Panner(Coordinate=Multiply(A=@n43, B=@n44), Time=@n46) {speedX=0.0702}) {texture=/Game/Effects/Fort_Effects/Textures/Outlander/T_InTheZone_LightRays_01}
  n48 = TextureCoordinate()
  n49 = ScalarParameter() {defaultValue=2.5, parameterName=Gradient Tile B}
  n50 = TextureSample(UVs=Panner(Coordinate=Multiply(A=@n48, B=@n49), Time=@n46) {speedX=-0.047}) {texture=&T_InTheZone_LightRays_01}
  n51 = TextureCoordinate() {uTiling=0.33, vTiling=0.01}
  n52 = TextureSample(UVs=Panner(Coordinate=Add(A=@n51, B=@n45), Time=@n46) {speedX=0.0436}) {texture=&T_NoisePack_01}
  n53 = ScalarParameter() {defaultValue=3, parameterName=Mask Brightness}
  n54 = TextureCoordinate()
  n55 = ScalarParameter() {defaultValue=-0.03, parameterName=Slide}
  n56 = OneMinus(None=ComponentMask(None=Add(A=@n54, B=@n55)) {g=true})
  n57 = ScalarParameter() {defaultValue=8, desc=Orig: 10.0, parameterName=Center Exponent}
  n58 = Add(
   A=Multiply(A=Multiply(A=@n47.R, B=@n50.G), B=Multiply(A=@n52.R, B=@n53))
   B=Multiply(A=Power(Base=@n56, Exp=@n57), B=ScalarParameter() {defaultValue=2, parameterName=Center Brightness})
  )
  n59 = Ceil(None=Add(A=@n58, B=ScalarParameter() {defaultValue=-0.7, desc=Orig: -0.7, parameterName=Ceil Bias}))
  n60 = ScalarParameter() {defaultValue=3, parameterName=Ceil Brightness}
  Multiply(
   A=LinearInterpolate(A=@n58, B=Multiply(A=@n59, B=@n60), Alpha=ScalarParameter() {defaultValue=0.5, parameterName=Style Lerp})
   B=Power(Base=VertexColor().R, Exp=ScalarParameter() {defaultValue=4, desc=Orig: 4.0, parameterName=Fake Fresnel Exponent})
  )
  n61 = VertexColor()
  n62 = MaterialFunctionCall(Alpha=@n15.A, Color Input 0=@n13, Color Input 1=@n14) {materialFunction=&DynamicBranch}
  n63 = MaterialFunctionCall(
   Alpha=VertexColor().A
   Color Input 0=MaterialFunctionCall!(
    WorldPosition
    PivotLocation
    WS Normals
    Custom Object Basis 1
    Custom Object Basis 2=VectorParameter() {parameterName=RotationA, defaultValue=<rgba 0 1 0 1>}.RGB
    Custom Object Basis 3
   ) {materialFunction=/Engine/Functions/Engine_MaterialFunctions02/WorldPositionOffset/AlignMeshToTheCamera}.World Position Offset
   Color Input 1=ScalarParameter() {parameterName=CamFacing}
  ) {materialFunction=&DynamicBranch}
  n64 = MaterialFunctionCall(
   Fade_Start_Distance=ScalarParameter() {defaultValue=1400, parameterName=End Fade}
   Fade_End_Distance=ScalarParameter() {defaultValue=1200, parameterName=Start Fade}
  ) {materialFunction=&MF_Distance_SyncedFade_FX}
  n65 = CameraPositionWS()
  n66 = ActorPositionWS()
  n67 = Multiply(A=Normalize(VectorInput=Subtract(A=@n65, B=@n66)), B=ScalarParameter() {desc=was -10.0, parameterName=Camera Facing Push})
  n68 = MaterialFunctionCall(
   Alpha=VertexColor().G
   Color Input 0=LinearInterpolate(A=@n62.Result, B=@n63.Result, Alpha=@n64.Result)
   Color Input 1=LinearInterpolate(A=Add(A=@n63.Result, B=@n67), B=@n62.Result, Alpha=@n64.Result)
  ) {materialFunction=&DynamicBranch}
  n69 = ActorPositionWS()
  n70 = WorldPosition()
  n71 = MaterialFunctionCall(
   Alpha=ScalarParameter() {parameterName=IsLegendaryRarity}
   Color Input 0=MaterialFunctionCall(Alpha=VertexColor().A, Color Input 0=@n68.Result, Color Input 1=Subtract(A=@n69, B=@n70.XYZ)) {materialFunction=&DynamicBranch}.Result
   Color Input 1=@n68.Result
  ) {materialFunction=&DynamicBranch}
  n72 = Add(A=@n71.Result, B=WorldPosition() {worldPositionShaderOffset=WPT_ExcludeAllShaderOffsets}.XYZ)
  n73 = CameraPositionWS()
  n74 = Multiply(A=Normalize(VectorInput=Subtract(A=@n72, B=@n73)), B=ScalarParameter() {defaultValue=-90, parameterName=MoveTowardsCamera})
  Multiply(A=OneMinus(None=@n61.A), B=@n74)
  n75 = ComponentMask(None=TextureCoordinate()) {r=true}
  n76 = ScalarParameter() {defaultValue=0.75, parameterName=StarTile}
  n77 = Multiply(A=@n75, B=Multiply(A=ScalarParameter() {defaultValue=2, parameterName=Gradient Tile A}, B=@n76))
  n78 = Multiply(A=@n75, B=Multiply(A=ScalarParameter() {defaultValue=2.5, parameterName=Gradient Tile B}, B=@n76))
  n79 = NamedRerouteDeclaration(None=ScalarParameter() {parameterName=Random}) {name=Random, nodeColor=<rgba 0.5703 1 0 1>}
  n80 = ScalarParameter() {defaultValue=2.8, parameterName=Panner Speed}
  n81 = Add(A=Time(), B=@n79)
  n82 = Add(
   A=AppendVector(A=AppendVector(A=@n77, B=@n78), B=Add(A=@n75, B=@n79))
   B=Multiply(A=Constant3Vector() {constant=<rgba 0.0786 -0.047 0.0436 0>}, B=Multiply(A=@n80, B=@n81))
  )
  n83 = Frac(None=Multiply(A=Sine(None=@n82)) {desc=Orig: 1.0})
  n84 = MaterialFunctionCall(Float3=Subtract(A=Multiply(A=@n83) {constB=2})) {materialFunction=&BreakOutFloat3Components}
  n85 = ScalarParameter() {defaultValue=3, parameterName=Mask Brightness}
  n86 = Add(A=VertexColor().R, B=ScalarParameter() {defaultValue=-0.03, parameterName=Slide})
  n87 = ScalarParameter() {defaultValue=8, desc=Orig: 10.0, parameterName=Center Exponent}
  n88 = Add(
   A=Multiply(A=Multiply(A=@n84.R, B=@n84.R), B=Multiply(A=@n84.B, B=@n85))
   B=Multiply(A=Power(Base=@n86, Exp=@n87), B=ScalarParameter() {defaultValue=2, parameterName=Center Brightness})
  )
  n89 = ScalarParameter() {defaultValue=-0.7, desc=Orig: -0.7, parameterName=Ceil Bias}
  n90 = Multiply(A=Ceil(None=Add(A=@n88, B=@n89)), B=ScalarParameter() {defaultValue=3, parameterName=Ceil Brightness})
  n91 = ScalarParameter() {defaultValue=0.5, parameterName=Style Lerp}
  n92 = VertexColor()
  n93 = ScalarParameter() {defaultValue=4, desc=Orig: 4.0, parameterName=Fake Fresnel Exponent}
  NamedRerouteDeclaration(None=Multiply(A=LinearInterpolate(A=@n88, B=@n90, Alpha=@n91), B=Power(Base=@n92.R, Exp=@n93))) {name=SphericalStar, nodeColor=<rgba 0 1 0.7656 1>}
  n94 = ScalarParameter() {defaultValue=2.25, parameterName=BG Boost}
  n95 = ScalarParameter() {defaultValue=0.25, parameterName=TOD_Morning}
  n96 = ScalarParameter() {defaultValue=0.3, parameterName=TOD_Day}
  n97 = ScalarParameter() {defaultValue=0.25, parameterName=TOD_Evening}
  n98 = ScalarParameter() {defaultValue=0.17, parameterName=TOD_Night}
  n99 = Distance(A=TextureCoordinate(), B=Constant() {r=0.5})
  n100 = ScalarParameter() {defaultValue=2, parameterName=GlowSize}
  n101 = Saturate(None=OneMinus(None=Multiply(A=@n99, B=@n100) {constB=2}))
  n102 = Constant() {r=1}
  n103 = DepthFade(FadeDistance=ScalarParameter() {defaultValue=25, parameterName=Depth Fade Distance})
  n104 = Multiply(
   A=Multiply(
    A=MaterialFunctionCall(Value_Morning=@n95, Value_Day=@n96, Value_Evening=@n97, Value_Night=@n98) {materialFunction=&Scalar_Time_of_Day}.Result
    B=Multiply(A=@n101, B=@n101)
   )
   B=QualitySwitch(Default=@n102, Low=@n102, High=@n103, Medium=@n103, Epic=@n103)
  )
  n105 = LinearInterpolate(Alpha=@n40) {constA=1, constB=0.15}
  n106 = ScalarParameter() {defaultValue=0.17, parameterName=BandWidth}
  n107 = Multiply(A=@n106) {constB=0.2}
  n108 = Multiply(
   A=MaterialFunctionCall(
    Width=LinearInterpolate(A=@n106, B=@n107, Alpha=@n40)
    Sharpness=ScalarParameter() {defaultValue=0.1, parameterName=BandSharpness}
    Offset=@n39
    Input Coordinates=@n41.Rotated Values
   ) {materialFunction=&GeneratedBand}.Result
   B=@n42.Result
  )
  n109 = Multiply(A=Multiply(A=@n104, B=Multiply(A=@n105, B=@n108))) {constB=5}
  n110 = ScalarParameter() {parameterName=UseBand}
  NamedRerouteDeclaration(None=Add(A=Multiply(A=@n94, B=@n104), B=LinearInterpolate(B=@n109, Alpha=@n110))) {name=HolographicMesh, nodeColor=<rgba 0.6172 0 1 1>}
  n111 = Distance(A=TextureCoordinate(), B=Constant() {r=0.5})
  n112 = ScalarParameter() {defaultValue=2, parameterName=GlowSize}
  n113 = Saturate(None=OneMinus(None=Multiply(A=@n111, B=@n112) {constB=2}))
  n114 = Constant() {r=1}
  n115 = DepthFade(FadeDistance=ScalarParameter() {defaultValue=25, parameterName=Depth Fade Distance})
  NamedRerouteDeclaration(None=Multiply(A=Multiply(A=@n113, B=@n113), B=QualitySwitch(Default=@n114, Low=@n114, High=@n115, Medium=@n115, Epic=@n115))) {name=DarkBG, nodeColor=<rgba 0 0.1562 1 1>}
  n116 = MaterialFunctionCall(Alpha=@n61.A, Color Input 0=Constant(), Color Input 1=@n74) {materialFunction=&DynamicBranch}
  n117 = CameraPositionWS()
  n118 = ObjectPositionWS()
  n119 = Saturate(None=Divide(A=Distance(A=@n117, B=@n118), B=ScalarParameter() {defaultValue=300, parameterName=NearDistanceClamp}))
  ShadingPathSwitch(Default=Add(A=@n71.Result, B=LinearInterpolate(B=@n116.Result, Alpha=@n119)), Mobile=@n71.Result) {desc=Migrated from FeatureLevelSwitch}
  n120 = TextureCoordinate()
  n121 = Constant() {desc=Change this to change the shape of the rays, r=10}
  n122 = Multiply(A=Frac(None=ComponentMask(None=@n120) {r=true}), B=@n121) {constB=10}
  n123 = Floor(None=@n122)
  n124 = Add(A=@n123)
  n125 = Multiply(A=AppendVector(A=Fmod(A=@n123, B=@n121), B=Fmod(A=@n124, B=@n121)), B=Divide(B=@n121)) {constB=0.1}
  n126 = MaterialFunctionCall(Float2=Frac(None=Sine(None=@n125))) {materialFunction=&BreakOutFloat2Components}
  n127 = Frac(None=@n122)
  n128 = Multiply(A=@n127) {constB=0.5}
  n129 = Multiply(A=OneMinus(None=Cosine(None=@n128))) {constB=0.5}
  SmoothStep(Value=LinearInterpolate(A=@n126.R, B=@n126.R, Alpha=QualitySwitch(Default=@n129, Low=@n127)))
 ]
}
