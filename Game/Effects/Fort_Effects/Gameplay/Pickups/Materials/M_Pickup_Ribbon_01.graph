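# Unlit additive material for pickup ribbon effects, enabled for beam trails.
# It multiplies the V texture coordinate by its inverse and scales by four, forming a soft band that peaks at the center and fades at both edges without textures.
# Particle color RGB tints this band, which then feeds the shared MF_FX_EmissiveAdjustments function as emissive.
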Material /Game/Effects/Fort_Effects/Gameplay/Pickups/Materials/M_Pickup_Ribbon_01
graph {
 properties {bUsedWithBeamTrails=true, blendMode=BLEND_Additive, pixelDepthOffsetMode=PDOM_Legacy, shadingModel=MSM_Unlit, usedShadingModels=Unlit}
 nodes [
  n1 = ComponentMask(None=TextureCoordinate()) {g=true}
  n2 = Multiply(A=@n1, B=OneMinus(None=@n1))
  MaterialFunctionCall(Emissive=Multiply(A=ParticleColor().RGB, B=Multiply(A=@n2) {constB=4})) {materialFunction=&MF_FX_EmissiveAdjustments}
 ]
}
