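# Material function that gives pickups a glowing, scanlined blueprint hologram look, returned as material attributes.
# Emissive combines Fresnel, animated edges from offset normal-map differencing and a bounding-box sweep, with back faces in flat stroke color.
# TimestampA transitions brightness, Fresnel falloff and normal push, while TimestampB fades opacity out.
# World position offset combines distance-scaled normal push with LOD cull and black-hole effects.
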MaterialFunction /Game/Effects/Fort_Effects/Gameplay/Pickups/Materials/M_Pickup_Blueprint_Function
graph {nodes [
  n1 = Add(A=Time(), B=ScalarParameter() {parameterName=Random})
  n2 = ScalarParameter() {defaultValue=4, parameterName=Fresnel Exponent A}
  n3 = ScalarParameter() {defaultValue=3, parameterName=Fresnel Exponent B}
  n4 = MaterialFunctionCall(
   TimestampA=ScalarParameter() {group=Timestamp, parameterName=TimestampA}
   TransitionLength_In=ScalarParameter() {defaultValue=-1, group=Timestamp, parameterName=TransitionLength_In}
  ) {materialFunction=&MF_Timestamp_InOut_01}
  n5 = TextureSampleParameter2D() {group=Textures, parameterName=Normals, samplerType=SAMPLERTYPE_Normal, texture=/Game/Weapons/FORT_Rifles/Textures/Death_Stalker_N}
  n6 = Transform(None=@n5.RGB)
  n7 = LinearInterpolate(
   A=ScalarParameter() {defaultValue=0.015, parameterName=Opacity Inner}
   B=ScalarParameter() {defaultValue=1.7, parameterName=Opacity Outer}
   Alpha=MaterialFunctionCall(
    Brightness=ScalarParameter() {defaultValue=1, parameterName=Fresnel Brightness}
    Exponent=LinearInterpolate(A=@n2, B=@n3, Alpha=@n4."SmoothStep (A Only)")
    Vertex / Pixel Normal=FunctionInput(Preview=@n6) {bUsePreviewValueAsDefault=true, inputName=World Space Normal}
   ) {materialFunction=&MF_Fresnel_FX}.Fresnel Inverted
  )
  n8 = VectorParameter() {parameterName=Color, defaultValue=<rgba 0.1 0.0705 0.044 1>}
  n9 = TextureSampleParameter2D(UVs=Add(A=ScalarParameter() {defaultValue=0.002, parameterName=Offset U}, B=TextureCoordinate())) {group=Textures, parameterName=Normals, samplerType=SAMPLERTYPE_Normal, texture=&Death_Stalker_N}
  n10 = FunctionInput(Preview=Subtract(A=@n9.RGB, B=@n5.RGB)) {bUsePreviewValueAsDefault=true, inputName=Vector 3 Normal Sobel}
  n11 = DotProduct(A=@n10, B=Constant() {r=1})
  n12 = ScalarParameter() {defaultValue=-0.25, parameterName=Mask Bias}
  n13 = MaterialFunctionCall(Float2=TextureCoordinate()) {materialFunction=&BreakOutFloat2Components}
  n14 = Multiply(A=ComponentMask(None=@n10) {r=true}, B=@n13.R)
  n15 = Multiply(A=@n1, B=ScalarParameter() {defaultValue=0.2249, parameterName=Edge Speed U})
  n16 = Power(Base=Sine(None=Add(A=@n14, B=@n15)) {period=0.7379}, Exp=ScalarParameter() {defaultValue=4, parameterName=Sine Edge Exponent U})
  n17 = Multiply(A=ComponentMask(None=@n10) {g=true}, B=@n13.G)
  n18 = Multiply(A=@n1, B=ScalarParameter() {defaultValue=0.1484, parameterName=Edge Speed V})
  n19 = Power(Base=Sine(None=Add(A=@n17, B=@n18)) {period=0.4491}, Exp=ScalarParameter() {defaultValue=6, parameterName=Sine Edge Exponent V})
  n20 = Multiply(A=Ceil(None=Add(A=@n11, B=@n12)), B=Clamp(None=Add(A=@n16, B=@n19)))
  n21 = ScalarParameter() {defaultValue=4, parameterName=Blueprint Edge Exponent}
  n22 = Multiply(A=@n1, B=ScalarParameter() {defaultValue=0.25, parameterName=Sweep Speed})
  n23 = Add(
   A=FunctionInput(Preview=MaterialFunctionCall() {materialFunction=&BoundingBoxBased_0-1_UVW}.G) {bUsePreviewValueAsDefault=true, inputName=Bounding box G, inputType=FunctionInput_Scalar, sortPriority=3}
   B=ConstantBiasScale(None=Frac(None=@n22)) {bias=-0.5, scale=3}
  )
  n24 = OneMinus(None=@n23)
  n25 = Clamp(None=Multiply(A=Multiply(A=@n23, B=@n24)) {constB=4})
  n26 = ScalarParameter() {defaultValue=80, parameterName=Sweep Exponent}
  n27 = Add(
   A=Multiply(A=Power(Base=@n20, Exp=@n21), B=ScalarParameter() {defaultValue=3, parameterName=Blueprint Edge Brightness})
   B=Multiply(A=Power(Base=@n25, Exp=@n26), B=ScalarParameter() {defaultValue=10, parameterName=Sweep Brightness})
  )
  n28 = Multiply(
   A=MaterialFunctionCall!(
    Number of Scanlines
    Scanline Speed
    "Scanlines Thickness (Bias)"
    Scanlines Dark Value
    Scanlines Light Value
    Block Opacity
    UVs
    Time=@n1
    Block Repeat
   ) {materialFunction=/Game/Effects/Fort_Effects/Effects/Atlas/Materials/MF_Atlas_Scanlines_01}.Result
   B=Add(A=Multiply(A=@n7, B=@n8.RGB), B=Multiply(A=@n8.RGB, B=@n27))
  )
  n29 = If(
   A=TwoSidedSign()
   B=Constant()
   "A > B"=@n28
   "A == B"=@n28
   "A < B"=Multiply(A=@n8.RGB, B=ScalarParameter() {defaultValue=3, parameterName=Stroke Brightness})
  )
  n30 = Multiply(A=@n29, B=ScalarParameter() {defaultValue=30, parameterName=Pickup Brightness})
  n31 = MaterialFunctionCall(
   TimestampA=ScalarParameter() {group=Timestamp, parameterName=TimestampB}
   TransitionLength_In=ScalarParameter() {defaultValue=-1, group=Timestamp, parameterName=TransitionLength_Out}
  ) {materialFunction=&MF_Timestamp_InOut_01}
  n32 = VectorParameter() {parameterName=Model Push Min, defaultValue=<rgba 0.5 0.5 0.5 0>}
  n33 = VectorParameter() {parameterName=Model Push Max, defaultValue=<rgba 1.2 1.2 1.2 0>}
  n34 = MaterialFunctionCall(
   Fade_Start_Distance=ScalarParameter() {defaultValue=1000, parameterName=Start}
   Fade_End_Distance=ScalarParameter() {defaultValue=300, parameterName=End}
  ) {materialFunction=&MF_Distance_SyncedFade_FX}
  n35 = Multiply(
   A=FunctionInput(Preview=VertexNormalWS()) {bUsePreviewValueAsDefault=true, inputName=Vertex Normal WS}
   B=LinearInterpolate(
    A=LinearInterpolate(A=@n32.RGB, B=@n33.RGB, Alpha=@n34.Result)
    B=Constant3Vector() {constant=<rgba 0.1 0.1 0.1 0>}
    Alpha=@n4."SmoothStep (A Only)"
   )
  )
  n36 = VectorParameter() {parameterName=WorldPostionOfBlackHole, defaultValue=<rgba -30000 25536 1952 0>}
  n37 = WorldPosition()
  n38 = MaterialFunctionCall!(
   World Position=Add(A=@n35, B=@n37.XYZ)
   DistanceCullFadeAnimationOffset=MaterialFunctionCall() {materialFunction=/Game/Packages/Fortress_SharedMaterials/Base_Material_Functions/PickupEffectDistanceCullOverride}.Result
   Use actor center instead of pivot point
  ) {materialFunction=/Game/Packages/Fortress_SharedMaterials/Base_Material_Functions/LODEffect}
  n39 = Add(
   A=MaterialFunctionCall!(WorldPostionOfBlackHole=AppendVector(A=@n36.RGB, B=@n36.RGB), World Position=Add(A=@n38.Result, B=@n37.XYZ)) {materialFunction=/Game/Packages/Fortnite_UniqueMaterials_1/BlackHoleFunction}.Result
   B=@n38.Result
  )
  FunctionOutput(None=MakeMaterialAttributes!(
    BaseColor
    Metallic
    Specular
    Roughness
    Anisotropy
    EmissiveColor=LinearInterpolate(A=@n29, B=@n30, Alpha=@n4."SmoothStep (A Only)")
    Opacity=OneMinus(None=@n31.SmoothStep)
    OpacityMask
    Normal
    Tangent
    WorldPositionOffset=Add(A=@n35, B=@n39)
    SubsurfaceColor
    ClearCoat
    ClearCoatRoughness
    AmbientOcclusion
    Refraction
    CustomizedUVs_0
    CustomizedUVs_1
    CustomizedUVs_2
    CustomizedUVs_3
    CustomizedUVs_4
    CustomizedUVs_5
    CustomizedUVs_6
    CustomizedUVs_7
    PixelDepthOffset
    ShadingModel
    Displacement
   ))
  ScalarParameter() {defaultValue=1, parameterName=DebugLERP}
  ScalarParameter() {parameterName=Fake Time}
 ]}
